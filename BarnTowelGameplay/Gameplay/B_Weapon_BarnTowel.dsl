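; BarnTowel jetpack weapon Blueprint that syncs flight animation state across the network.
; Mesh-rotation and speed updates go through change-checked setters into replicated copies, which OnRep functions apply only on simulated proxies.
; On the server, flying start/end events run a timer replicating control rotation at a console-variable interval.
; It also counts trigger-hold charge, defines a camera sphere trace and exposes movement input, hover state and rotation via GetJetpackData.
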(event EventHandleMeshRotationValuesUpdated
  (params (GoalMovementOrientation :type Rotator :default "0, 0, 0") (GoalMovementDirection :type Vector :default "0, 0, 0") (CurrentNoiseOffset :type Vector :default "0, 0, 0") (DodgeVelocity :type Vector :default "0, 0, 0") (MovementInput :type Vector%202D%20Structure) (bIsStartingFlying :type Boolean) (bIsinInitialHover :type Boolean) (bIsInInitialHoverWithBuffer :type Boolean) (TimeLeftInInitialHover :type Float%20%28single-precision%29))
  (CallFunction|SetforAnimGoalMoveOrientation GoalMovementOrientation)
  (CallFunction|SetforAnimMovementGoalDirection GoalMovementDirection)
  (CallFunction|SetforAnimCurrentNoiseOffset CurrentNoiseOffset)
  (CallFunction|SetforAnimDodgeVelocity DodgeVelocity)
  (CallFunction|SetforAnimMovementInput MovementInput)
  (CallFunction|SetforAnimIsStartingFlying bIsStartingFlying)
  (CallFunction|SetforAnimIsinInitialHover bIsinInitialHover)
  (CallFunction|SetforAnimIsinInitialHoverwithBuffer bIsInInitialHoverWithBuffer))

(event EventBeginPlay
  (bind _returnvalue (Game|GetInstigator))
  (|CallParentFunction)
  (CallFunction|BindonHandFired)
  (if (Networking|IsServer)
    (bind _returnvalue_2 (Development|GetConsoleVariableFloatValue "Fort.BarnTowel.ReplicateControlRotationInterval"))
    (if (> _returnvalue_2 0)
      ; Replicate control rotation when flying
      (Variables|Default|SetReplicateRotationInterval _returnvalue_2)
      (Ability|Async|WaitGameplayEventToActor _returnvalue "(TagName=\"Event.Athena.BarnTowel.Flying.Start\")"
        (:then
          (Ability|Async|WaitGameplayEventToActor _returnvalue "(TagName=\"Event.Athena.BarnTowel.Flying.End\")"
            (:EventReceived
              (Utilities|Time|ClearandInvalidateTimerbyHandle (Variables|Default|GetReplicateRotationTimer)))))
        (:EventReceived
          (bind _returnvalue_3 (Utilities|Time|SetTimerbyEvent (delegate Custom|ReplicateRotation) (Variables|Default|GetReplicateRotationInterval) true true))
          (Variables|Default|SetReplicateRotationTimer _returnvalue_3))))))

(event EventHandleForwardSpeedUpdated
  (params (NewForwardSpeed :type Float%20%28single-precision%29))
  (CallFunction|SetForwardSpeed NewForwardSpeed))

(event Custom|ReplicateRotation
  ; Replicate control rotation when flying
  (bind _asfort_pawn (Utilities|Casting|CastToFortPawn (Game|GetInstigator)))
  (Variables|Default|SetReplicatedControlRotation (Pawn|GetControlRotation _asfort_pawn :!class /Script/Engine.Pawn)))

(event Custom|BindOnHandFired
  ;[1 1 1] On hand fired (left - right)
  (EventDispatchers|BindEventtoOnPressTrigger (delegate Custom|StartedCharging))
  (EventDispatchers|BindEventtoOnReleaseTrigger (delegate Custom|StoppedCharging)))

(event Custom|StartedCharging
  (params (Weapon :type Fort%20Weapon%20Object%20Reference))
  ;[1 1 1] On hand fired (left - right)
  (bind _returnvalue (Utilities|Time|SetTimerbyEvent (delegate Custom|AddToCharge) 0.05 true true))
  (Variables|Default|SetChargingTimer _returnvalue))

(event Custom|StoppedCharging
  (params (Weapon :type Fort%20Weapon%20Object%20Reference))
  ;[1 1 1] On hand fired (left - right)
  (Utilities|Time|ClearandInvalidateTimerbyHandle (Variables|Default|GetChargingTimer))
  (Variables|Default|SetCharge 0.0))

(event Custom|HandFired
  (params (Hand :type EDualWeaponHand%20Enum :default "LEFT") (bPersistantFire :type Boolean)))

(event Custom|PressedTrigger_ToServer
  (Utilities|FlowControl|DoOnce false))

(event Custom|AddToCharge
  ;[1 1 1] On hand fired (left - right)
  (Math|Float|IncrementFloat (Variables|Default|GetCharge)))

(event Custom|OnPressedTrigger
  (params (Weapon :type Fort%20Weapon%20Object%20Reference))
  (CallFunction|PressedTriggertoServer)
  (Utilities|FlowControl|DoOnce false))

(event EventOnPlayWeaponFireFX
  (params (bPersistentFire :type Boolean) (bSecondaryFire :type Boolean))
  (|CallParentFunction bPersistentFire bSecondaryFire))

(event Custom|AllowNextFire
  (Utilities|FlowControl|DoOnce false))

(event EventOnStopWeaponFireFX
  (FortWeapon|StopForceFeedback))

(fn DoAOETrace
  (bind _outhits (Collision|MultiSphereTraceByProfile (Camera|GetCameraLocation (Variables|/Script/Engine.PlayerController|GetPlayerCameraManager (Fortnite|GetFortPlayerControllerfromActor (Variables|Actor|GetInstigator)))) (+ (Camera|GetCameraLocation (Variables|/Script/Engine.PlayerController|GetPlayerCameraManager (Fortnite|GetFortPlayerControllerfromActor (Variables|Actor|GetInstigator)))) (* (Math|Vector|GetForwardVector (Pawn|GetBaseAimRotation (Variables|Actor|GetInstigator))) 5e+03)) 1.2e+02 "OverlapAll" false (Utilities|Array|MakeArray (Variables|Actor|GetInstigator))))
  (return _outhits))

(fn GetJetpackData
  (return (Variables|Default|GetForAnim_MovementInputBP) (Variables|Default|GetForAnim_IsInInitialHover) (Variables|Default|GetReplicatedControlRotation)))

(fn OnRep_ReplicatedCurrentNoiseOffset
  (if (Utilities|Enum|Equal(Enum) (Networking|GetLocalRole (Variables|Actor|GetInstigator)) "ROLE_SimulatedProxy")
    (Variables|Default|SetForAnim_CurrentNoiseOffset (Variables|Default|GetReplicatedCurrentNoiseOffset))))

(fn OnRep_ReplicatedDodgeVelocity
  (if (Utilities|Enum|Equal(Enum) (Networking|GetLocalRole (Variables|Actor|GetInstigator)) "ROLE_SimulatedProxy")
    (Variables|FortWeaponJetpack|SetForAnimDodgeVelocity (Variables|Default|GetReplicatedDodgeVelocity))))

(fn OnRep_ReplicatedForwardSpeed
  (if (Utilities|Enum|Equal(Enum) (Networking|GetLocalRole (Variables|Actor|GetInstigator)) "ROLE_SimulatedProxy")
    (Variables|FortWeaponJetpack|SetForwardSpeed (Variables|Default|GetReplicatedForwardSpeed))))

(fn OnRep_ReplicatedGoalMoveOrientation
  (if (Utilities|Enum|Equal(Enum) (Networking|GetLocalRole (Variables|Actor|GetInstigator)) "ROLE_SimulatedProxy")
    (Variables|FortWeaponJetpack|SetForAnimGoalMoveOrientation (Variables|Default|GetReplicatedGoalMoveOrientation))))

(fn OnRep_ReplicatedIsInInitialHover
  (if (Utilities|Enum|Equal(Enum) (Networking|GetLocalRole (Variables|Actor|GetInstigator)) "ROLE_SimulatedProxy")
    (Variables|Default|SetForAnim_IsInInitialHover (Variables|Default|GetReplicatedIsInInitialHover))))

(fn OnRep_ReplicatedIsInInitialHoverWithBuffer
  (if (Utilities|Enum|Equal(Enum) (Networking|GetLocalRole (Variables|Actor|GetInstigator)) "ROLE_SimulatedProxy")
    (Variables|Default|SetForAnim_IsInInitialHover_WithBuffer (Variables|Default|GetReplicatedIsInInitialHoverWithBuffer))))

(fn OnRep_ReplicatedIsStartingFlying
  (if (Utilities|Enum|Equal(Enum) (Networking|GetLocalRole (Variables|Actor|GetInstigator)) "ROLE_SimulatedProxy")
    (Variables|FortWeaponJetpack|SetbForAnimIsStartingFlying (Variables|Default|GetReplicatedIsStartingFlying))))

(fn OnRep_ReplicatedMovementGoalDirection
  (if (Utilities|Enum|Equal(Enum) (Networking|GetLocalRole (Variables|Actor|GetInstigator)) "ROLE_SimulatedProxy")
    (Variables|Default|SetForAnim_MovementGoalDirection (Variables|Default|GetReplicatedMovementGoalDirection))))

(fn OnRep_ReplicatedMovementInput
  (if (Utilities|Enum|Equal(Enum) (Networking|GetLocalRole (Variables|Actor|GetInstigator)) "ROLE_SimulatedProxy")
    (Variables|Default|SetForAnim_MovementInputBP (Variables|Default|GetReplicatedMovementInput))))

(fn OnRep_ReplicatedTimeLeftInInitialHover
  (if (Utilities|Enum|Equal(Enum) (Networking|GetLocalRole (Variables|Actor|GetInstigator)) "ROLE_SimulatedProxy")
    (Variables|FortWeaponJetpack|SetForAnimTimeLeftInInitialHover (Variables|Default|GetReplicatedTimeLeftInInitialHover))))

(fn SetForAnimMovementGoalDirection
  (params (NewMovementGoalDirection :type Vector))
  (if (Utilities|Operators|NotEqual(!=) NewMovementGoalDirection (Variables|Default|GetForAnim_MovementGoalDirection) 0)
    (bind _output_get (Variables|Default|SetForAnim_MovementGoalDirection NewMovementGoalDirection))
    (Variables|Default|SetReplicatedMovementGoalDirection _output_get)))

(fn SetForwardSpeed
  (params (NewForwardSpeed :type Float%20%28double-precision%29))
  (if (!= NewForwardSpeed (Variables|FortWeaponJetpack|GetForwardSpeed))
    (Variables|FortWeaponJetpack|SetForwardSpeed NewForwardSpeed)
    (Variables|Default|SetReplicatedForwardSpeed NewForwardSpeed)))

(fn SetForAnimCurrentNoiseOffset
  (params (NewNoiseOffset :type Vector))
  (if (Utilities|Operators|NotEqual(!=) NewNoiseOffset (Variables|Default|GetForAnim_CurrentNoiseOffset) 0)
    (bind _output_get (Variables|FortWeaponJetpack|SetForAnimCurrentNoiseOffset NewNoiseOffset))
    (Variables|Default|SetReplicatedCurrentNoiseOffset _output_get)))

(fn SetForAnimDodgeVelocity
  (params (New%20Dodge%20Velocity :type Vector))
  (if (Utilities|Operators|NotEqual(!=) New%20Dodge%20Velocity (Variables|FortWeaponJetpack|GetForAnimDodgeVelocity) 1.0)
    (bind _output_get (Variables|FortWeaponJetpack|SetForAnimDodgeVelocity New%20Dodge%20Velocity))
    (Variables|Default|SetReplicatedDodgeVelocity _output_get)))

(fn SetForAnimGoalMoveOrientation
  (params (NewGoalMoveOrientation :type Rotator))
  (if (Utilities|Operators|NotEqual(!=) NewGoalMoveOrientation (Variables|Default|GetForAnim_GoalMoveOrientation) 0)
    (bind _output_get (Variables|Default|SetForAnim_GoalMoveOrientation NewGoalMoveOrientation))
    (Variables|Default|SetReplicatedGoalMoveOrientation _output_get)))

(fn SetForAnimIsInInitialHover
  (params (NewIsInInitialHover :type Boolean))
  (if (!= NewIsInInitialHover (Variables|Default|GetForAnim_IsInInitialHover))
    (bind _output_get (Variables|Default|SetForAnim_IsInInitialHover NewIsInInitialHover))
    (Variables|Default|SetReplicatedIsInInitialHover _output_get)))

(fn SetForAnimIsInInitialHoverWithBuffer
  (params (NewIsInInitialHoverWithBuffer :type Boolean))
  (if (!= NewIsInInitialHoverWithBuffer (Variables|Default|GetForAnim_IsInInitialHover_WithBuffer))
    (bind _output_get (Variables|Default|SetForAnim_IsInInitialHover_WithBuffer NewIsInInitialHoverWithBuffer))
    (Variables|Default|SetReplicatedIsInInitialHoverWithBuffer _output_get)))

(fn SetForAnimIsStartingFlying
  (params (NewIsStartingFlying :type Boolean))
  (if (!= NewIsStartingFlying (Variables|Default|GetForAnim_IsStartingFlying))
    (bind _output_get (Variables|Default|SetForAnim_IsStartingFlying NewIsStartingFlying))
    (Variables|Default|SetReplicatedIsStartingFlying _output_get)))

(fn SetForAnimMovementInput
  (params (NewMovementInput :type Vector%202D%20Structure))
  (if (Utilities|Operators|NotEqual(!=) NewMovementInput (Variables|Default|GetForAnim_MovementInputBP) 0)
    (bind _output_get (Variables|Default|SetForAnim_MovementInputBP NewMovementInput))
    (Variables|Default|SetReplicatedMovementInput _output_get)))

(fn SetForAnimTimeLeftInInitialHover
  (params (New%20Time%20Left%20In%20Initial%20Hover :type Float%20%28double-precision%29))
  (if (!= New%20Time%20Left%20In%20Initial%20Hover (Variables|FortWeaponJetpack|GetForAnimTimeLeftInInitialHover))
    (bind _output_get (Variables|FortWeaponJetpack|SetForAnimTimeLeftInInitialHover New%20Time%20Left%20In%20Initial%20Hover))
    (Variables|Default|SetReplicatedTimeLeftInInitialHover _output_get)))

(fn ConstructionScript
  (|CallParentFunction))
