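; Parent weapon Blueprint for BarnTowel that switches the HUD key action prompts according to whether the instigator is flying.
; On BeginPlay, it starts wait tasks for the Granted.Athena.BarnTowel.Flying gameplay tag being added to or removed from the instigator.
; Adding the tag swaps the first two key actions for the higher-indexed set, and removing it restores the first two.
; The incoming damage override and construction script are empty.
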(event EventBeginPlay
  (|CallParentFunction)
  (Ability|Tasks|WaitGameplayTagAddToActor (Game|GetInstigator) "(TagName=\"Granted.Athena.BarnTowel.Flying\")"
    (:then
      (Ability|Tasks|WaitGameplayTagRemoveFromActor (Game|GetInstigator) "(TagName=\"Granted.Athena.BarnTowel.Flying\")"
        (:Removed
          (for _index (range 2 11)
            (WeaponHUDData|SetHudKeyActionVisibility _index))
          (WeaponHUDData|SetHudKeyActionVisibility 0 true)
          (WeaponHUDData|SetHudKeyActionVisibility 1 true))))
    (:Added
      (for _index_2 (range 2 11)
        (WeaponHUDData|SetHudKeyActionVisibility _index_2 true))
      (WeaponHUDData|SetHudKeyActionVisibility 0)
      (WeaponHUDData|SetHudKeyActionVisibility 1))))

(fn BP_ModifyIncomingDamage
  (params (Damage :type Float%20%28single-precision%29) (DamageTags :type Gameplay%20Tag%20Container%20Structure%20%28by%20ref%29) (DamageHit :type Hit%20Result%20Structure%20%28by%20ref%29) (EventInstigator :type Controller%20Object%20Reference) (DamageCauser :type Actor%20Object%20Reference)))

(fn ConstructionScript)
