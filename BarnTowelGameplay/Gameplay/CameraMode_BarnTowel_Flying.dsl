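; Flying camera mode for BarnTowel that filters how view rotation is processed.
; Its BP_ProcessViewRotation function zeroes the roll delta and passes the incoming view rotation and the pitch and yaw deltas through unchanged.
; It returns false for override, so, per its comment, the final rotation is the view rotation plus those deltas and other modifiers.
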(fn BP_ProcessViewRotation
  (params (InViewRotation :type Rotator%20%28by%20ref%29 :default "0, 0, 0") (InDeltaRot_Roll :type Float%20%28single-precision%29) (InDeltaRot_Pitch :type Float%20%28single-precision%29) (InDeltaRot_Yaw :type Float%20%28single-precision%29) (DeltaTime :type Float%20%28single-precision%29))
  ;[1 1 1] when ShouldOverride is true, we directly apply the new view rotation; else, we return view rotation + delta rot + other modifiers
  (return false InViewRotation 0.0 InDeltaRot_Pitch InDeltaRot_Yaw))
