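; Camera modifier for the BarnTowel flight weapon that shifts view location and FOV as the player hovers, flies, boosts and thrusts.
; When native calculation is off, Activate timestamps thrust gameplay events and records anchor locations.
; BlueprintModifyCamera derives velocity and acceleration from frame deltas, interpolating state-dependent offsets and FOV, plus acceleration-scaled sine shake.
; Curve-weighted anchor points hold the camera back during lead-in, quick thrust and dive.
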(event Custom|Activate
  (if (CallFunction|UseNativeCalculation)
    (CallFunction|NativeActivate)
    (else
      (Variables|Default|SetStartedTime (Utilities|Time|GetGameTimeinSeconds))
      (Variables|Default|SetAnchoredLocation (Transformation|GetActorLocation (Variables|Default|GetPlayerPawn_Athena)))
      (Ability|Async|WaitGameplayEventToActor (Variables|Default|GetPlayerPawn_Athena) "(TagName=\"Event.Athena.BarnTowel.StartedWindingUpThrust\")"
        (:then
          (Ability|Async|WaitGameplayEventToActor (Variables|Default|GetPlayerPawn_Athena) "(TagName=\"Event.Athena.BarnTowel.StartedThrusting\")"
            (:then
              (Ability|Async|WaitGameplayEventToActor (Variables|Default|GetPlayerPawn_Athena) "(TagName=\"Event.Athena.Flying.DoInstantThrust\")"
                (:EventReceived
                  (Variables|Default|SetStartedQuickThrustTime (Utilities|Time|GetGameTimeinSeconds))
                  (Variables|Default|SetAnchoredLocation (Transformation|GetActorLocation (Variables|Default|GetPlayerPawn_Athena))))))
            (:EventReceived
              (Variables|Default|SetThrust_WindingUp false)
              (Variables|Default|SetTimeThrustStart (Utilities|Time|GetGameTimeinSeconds)))))
        (:EventReceived
          (Variables|Default|SetThrust_WindingUp true)
          (Variables|Default|SetThrust_WindingUpStartedTime (Utilities|Time|GetGameTimeinSeconds)))))))

(fn BlueprintModifyCamera
  (params (DeltaTime :type Float%20%28single-precision%29) (ViewLocation :type Vector :default "0, 0, 0") (ViewRotation :type Rotator :default "0, 0, 0") (FOV :type Float%20%28single-precision%29))
  (bind (_offset _interprate) (CallFunction|CalculateOffsetandInterpSpeed (Variables|GetViewRotation)))
  (bind (_target _interpspeed) (CallFunction|CalculateFOVTargetandInterpSpeed))
  (bind _returnvalue (Utilities|Time|GetGameTimeinSeconds))
  (bind _viewrotation (Variables|GetViewRotation))
  (bind _returnvalue_2 (Math|Vector|GetForwardVector _viewrotation))
  (bind _viewrotation_2 (Variables|GetViewRotation))
  (bind _returnvalue_3 (Math|Vector|GetRightVector _viewrotation_2))
  (bind _viewrotation_3 (Variables|GetViewRotation))
  (bind _returnvalue_4 (Math|Vector|GetUpVector _viewrotation_3))
  (bind _deltatime (Variables|GetDeltaTime))
  (bind _returnvalue_5 (Math|Float|NearlyEqual(Float) _deltatime))
  (bind _returnvalue_6 (Math|Float|SelectFloat 1.0 _deltatime _returnvalue_5))
  (bind _returnvalue_7 (Utilities|Time|GetGameTimeinSeconds))
  (bind (_isinstate _iswindingup) (Utilities|GetCrashingState))
  (bind (_isinstate_2 _iswindingup_2) (Utilities|GetCrashingState))
  (bind (_isinstate_3 _iswindingup_3 _winduppercent _divepercent) (Utilities|GetCrashingState))
  (bind _startedquickthrusttime (Variables|Default|GetStartedQuickThrustTime))
  (bind _returnvalue_8 (- _returnvalue _startedquickthrusttime))
  (bind _startedtime (Variables|Default|GetStartedTime))
  (bind _returnvalue_9 (- _returnvalue_7 _startedtime))
  (bind _quickthrustpullbacktime (Variables|Default|GetQuickThrustPullBackTime))
  (bind _foranim_movementinputbp_x (Math|Vector2D|BreakVector2D (Variables|/BarnTowelGameplay/Gameplay/B_Weapon_BarnTowel_Parent.B_Weapon_BarnTowel_Parent_C|GetForAnim_MovementInputBP_X (Variables|Default|GetWeapon))))
  (bind _velocity (LocalVariables|GetVelocity))
  (bind _winduphovertime (Variables|Default|GetWindUpHoverTime))
  (bind _acceleration (Variables|Default|GetAcceleration))
  (if (not (CallFunction|UseNativeCalculation))
    (Utilities|FlowControl|Sequence
      (:then_0
        (Utilities|IsValid (Variables|Default|GetPlayerPawn_Athena)
          (:Is%20Not%20Valid
            (bind _asplayer_pawn_athena (Utilities|Casting|CastToPlayerPawn_Athena (Game|Player|GetPlayerPawn (Utilities|Casting|CastToFortPlayerController (Game|Player|GetOwningPlayerController (Variables|CameraModifier|GetCameraOwner))))))
            (Variables|Default|SetPlayerPawn_Athena _asplayer_pawn_athena)
            (CallFunction|Activate)
            (bind _asb_weapon_barn_towel (Utilities|Casting|CastToB_Weapon_BarnTowel (Variables|/Script/FortniteGame.FortPawn|GetCurrentWeapon (Variables|Default|GetPlayerPawn_Athena))))
            (Variables|Default|SetWeapon _asb_weapon_barn_towel))))
      (:then_2
        (Variables|Default|SetIsHovering (GameplayTags|HasMatchingGameplayTag (Variables|Default|GetPlayerPawn_Athena) "(TagName=\"Granted.Athena.BarnTowel.Hovering\")"))
        (Variables|Default|SetAnchor_UpOffset 0.0)
        (Variables|Default|SetAnchor_ForwardOffset 0.0)
        (if (< _returnvalue_8 _quickthrustpullbacktime)
          (CallFunction|UseAnchorPoint (Variables|GetViewRotation) (Variables|Default|GetAnchoredLocation) (Variables|Default|GetQuickthrustAnchorPercentCurve) (Variables|Default|GetQuickthrustAnchorPercentCurve) (Utilities|NormalizeFloat _returnvalue_8 0 _quickthrustpullbacktime true))))
      (:then_3
        (if _isinstate_3
          (if _iswindingup_3
            (Variables|Default|SetAnchoredLocation (Transformation|GetActorLocation (Variables|Default|GetPlayerPawn_Athena)))
            (else
              (CallFunction|UseAnchorPoint (Variables|GetViewRotation) (Variables|Default|GetAnchoredLocation) (Variables|Default|GetThrustingAnchorPercentCurve) (Variables|Default|GetThrustingAnchorPercentCurve) _divepercent)))))
      (:then_4
        (if (CallFunction|IsinLeadIn)
          (if (< _returnvalue_9 _winduphovertime)
            (Variables|Default|SetLeadInAnchorLocation (Transformation|GetActorLocation (Variables|Default|GetPlayerPawn_Athena)))
            (Variables|Default|SetIsInInitialHover true)
            (Variables|Default|SetWasPressingForwardDuringInitialHover (> _foranim_movementinputbp_x 0.2))
            (else
              (CallFunction|UseAnchorPoint (Variables|GetViewRotation) (Variables|Default|GetLeadInAnchorLocation) (select (Variables|Default|GetWasPressingForwardDuringInitialHover) (Variables|Default|GetLeadIn_UpPercentCurve) (Variables|Default|GetLeadIn_UpPercentCurve_StraightUp)) (Variables|Default|GetLeadIn_ForwardPercentCurve) (Utilities|NormalizeFloat _returnvalue_9 _winduphovertime (Variables|Default|GetLeadInTime) true) true)
              (Variables|Default|SetIsInInitialHover false)))))
      (:then_5
        (if (or _iswindingup_2 (GameplayTags|HasMatchingGameplayTag (Variables|Default|GetPlayerPawn_Athena) "(TagName=\"Ability.HighTower.Power.Tomato.RepulsorCannon\")"))
          (if (not (Variables|Default|GetHasSavedSign))
            (Variables|Default|SetHasSavedSign true)
            (Variables|Default|SetSavedSign (Math|Float|Sign(Float) (Variables|Default|GetCurrentRight))))
          (else
            (Variables|Default|SetHasSavedSign false)
            (Variables|Default|SetSavedSign -1.0))))
      (:then_6
        (Variables|Default|SetIsBoosting (GameplayTags|HasMatchingGameplayTag (Variables|Default|GetPlayerPawn_Athena) "(TagName=\"Granted.Athena.BarnTowel.SmashingThings\")"))
        (bind _output_get (LocalVariables|SetVelocity (/ (- (Transformation|GetActorLocation (Variables|Default|GetPlayerPawn_Athena)) (Variables|Default|GetLastFramePosition_)) _returnvalue_6)))
        (Variables|Default|SetAcceleration (/ (- (LocalVariables|GetVelocity) (Variables|Default|GetLastFrameVelocity)) _returnvalue_6))
        (LocalVariables|SetDirection (Math|Vector|Normalize _output_get))
        (LocalVariables|SetSpeed (Math|Vector|VectorLength _output_get))
        (Variables|Default|SetLastFramePosition_ (Transformation|GetActorLocation (Variables|Default|GetPlayerPawn_Athena)))
        (Variables|Default|SetLastFrameVelocity (LocalVariables|GetVelocity))
        (Variables|Default|SetScreenShakeOffset (* (* (Math|Vector|Normalize (Math|Vector|ProjectVectorontoPlane (Variables|Default|GetAcceleration) (Math|Vector|GetForwardVector (Variables|GetViewRotation)))) (* (Math|Trig|Sin(Radians) (* (Utilities|Time|GetGameTimeinSeconds) 5e+01)) (Math|Float|SelectFloat 9.0 2.5 (and _isinstate (not _iswindingup))))) (* (Utilities|RemapFloat (Math|Vector|VectorLength (Variables|Default|GetAcceleration)) "(X=0.000000,Y=4000.000000)" "(X=0.000000,Y=1.000000)" true) (Utilities|RemapFloat (Math|Vector|VectorLength (LocalVariables|GetVelocity)) "(X=400.000000,Y=1400.000000)" "(X=0.000000,Y=1.000000)" true)))))
      (:then_7
        (Variables|Default|SetADS_Alpha 0.0)
        (if (CallFunction|IsinLeadIn)
          _)
        (LocalVariables|SetForwardSpeed (Math|Vector|DotProduct _velocity _returnvalue_2))
        (LocalVariables|SetRightSpeed (Math|Vector|DotProduct _velocity _returnvalue_3))
        (LocalVariables|SetUpSpeed (Math|Vector|DotProduct _velocity _returnvalue_4))
        (LocalVariables|SetUpAcceleration (Math|Vector|DotProduct _acceleration _returnvalue_4))
        (LocalVariables|SetRightAcceleration (Math|Vector|DotProduct _acceleration _returnvalue_3))
        (LocalVariables|SetForwardAcceleration (Math|Vector|DotProduct _acceleration _returnvalue_2))
        (LocalVariables|SetCurrentStraightness (Math|Float|Absolute(Float) (Math|Vector|DotProduct (Math|Vector|GetForwardVector (Variables|GetViewRotation)) (LocalVariables|GetDirection))))
        (Variables|Default|SetCurrentForward (Math|Interpolation|FInterpTo (Variables|Default|GetCurrentForward) (+ (* (Utilities|RemapFloat (Math|Vector|VectorLengthXY (LocalVariables|GetVelocity)) "(X=1100.000000,Y=2100.000000)" "(X=75.000000,Y=200.000000)" true) (Math|Float|SelectFloat 1.3 (Math|Float|SelectFloat 0.2 0.5 (Variables|Default|GetIsHovering)) (Variables|Default|GetIsBoosting))) (Utilities|RemapFloat (Math|Float|Absolute(Float) (LocalVariables|GetUpSpeed)) "(X=0.000000,Y=300.000000)" "(X=0.000000,Y=42.000000)" true)) (Variables|GetDeltaTime) (CallFunction|CalculateOffsetInterpSpeed)))
        (Variables|Default|SetCurrentRight (Math|Interpolation|FInterpTo (Variables|Default|GetCurrentRight) (+ (Utilities|Operators|Multiply (Utilities|RemapFloat (LocalVariables|GetRightSpeed) "(X=-1400.000000,Y=1400.000000)" "(X=-200.000000,Y=200.000000)" true) (Utilities|RemapFloat (Math|Float|Absolute(Float) (Variables|Default|GetCurrentForward)) "(X=50.000000,Y=300.000000)" "(X=1.000000,Y=1.300000)" true) (Math|Float|SelectFloat 1.3 (Math|Float|SelectFloat 0.3 1.0 (Variables|Default|GetIsHovering)) (Variables|Default|GetIsBoosting))) (* (Math|Float|SelectFloat 7e+01 0.0 (Variables|Default|GetIsHovering)) (Variables|Default|GetSavedSign))) (Variables|GetDeltaTime) (CallFunction|CalculateOffsetInterpSpeed)))
        (Variables|Default|SetCurrentUp (Math|Interpolation|FInterpTo (Variables|Default|GetCurrentUp) (* (+ (Utilities|Operators|Multiply (Utilities|RemapFloat (LocalVariables|GetUpSpeed) "(X=-1400.000000,Y=1400.000000)" "(X=-150.000000,Y=150.000000)" true) (Utilities|RemapFloat (Math|Float|Absolute(Float) (Variables|Default|GetCurrentForward)) "(X=50.000000,Y=400.000000)" "(X=1.000000,Y=1.600000)" true) (Math|Float|SelectFloat 1.3 1.0 (Variables|Default|GetIsBoosting)) (Math|Float|SelectFloat (Math|Float|SelectFloat 1.0 1.5 (CallFunction|IsinLeadIn)) 0.9 (> (Utilities|RemapFloat (LocalVariables|GetUpSpeed) "(X=-1400.000000,Y=1400.000000)" "(X=-150.000000,Y=150.000000)" true) 0))) (Math|Float|SelectFloat 4e+01 0.0 (Variables|Default|GetIsHovering))) (Math|Float|Clamp(Float) (Math|Vector|DotProduct (Math|Vector|GetUpVector (Variables|GetViewRotation)) "0.000000,0.000000,1.000000"))) (Variables|GetDeltaTime) (CallFunction|CalculateOffsetInterpSpeed)))
        (Variables|Default|SetCurrentLean 0.0)
        (Variables|Default|SetCurrentFOV (Math|Interpolation|FInterpTo (Variables|Default|GetCurrentFOV) _target (Variables|GetDeltaTime) _interpspeed))
        (Variables|Default|SetCurrentOffset (Math|Interpolation|FInterpTo (Variables|Default|GetCurrentForward) (Utilities|RemapFloat (LocalVariables|GetSpeed) "(X=-3500.000000,Y=3500.000000)" "(X=-220.000000,Y=220.000000)" true) (Variables|GetDeltaTime) 4.0))
        (Variables|Default|SetCurrentOffset_Vec (Math|Interpolation|VInterpTo (Variables|Default|GetCurrentOffset_Vec) _offset (Variables|GetDeltaTime) _interprate)))
      (:then_8
        (return (Utilities|Operators|Subtract (Variables|GetViewLocation) (+ (* (Utilities|Operators|Add (* (Variables|Default|GetCurrentOffset_Vec) (- 1.0 (Variables|Default|GetADS_Alpha))) (Utilities|Operators|Multiply (Math|Vector|GetForwardVector (Variables|GetViewRotation)) (- 1.0 (Variables|Default|GetADS_Alpha)) -3e+01) (* (Math|Vector|GetUpVector (Variables|GetViewRotation)) 3e+01) (* (Math|Vector|GetRightVector (Variables|GetViewRotation)) 0)) (Utilities|CalculateAlpha)) (+ (* (Math|Vector|GetForwardVector (Variables|GetViewRotation)) (Variables|Default|GetAnchor_ForwardOffset)) (* (Math|Vector|GetUpVector (Variables|GetViewRotation)) (Variables|Default|GetAnchor_UpOffset)))) (Variables|Default|GetScreenShakeOffset)) (Variables|GetViewRotation) (+ (Variables|GetFOV) (* (Variables|Default|GetCurrentFOV) (Utilities|CalculateAlpha))))))))

(fn CalculateFOVTargetAndInterpSpeed
  (bind (_isinstate _iswindingup) (Utilities|GetCrashingState))
  (if (CallFunction|IsinLeadIn)
    (if (Variables|Default|GetIsInInitialHover)
      (return -15.0 0.5)
      (else
        (return 2e+01 1e+01)))
    (elif (GameplayTags|HasMatchingGameplayTag (Variables|Default|GetPlayerPawn_Athena) "(TagName=\"Ability.HighTower.Power.Tomato.RepulsorCannon\")")
      (return -35.0 4.0)
      (elif _isinstate
        (if _iswindingup
          (return -5.0 5.0)
          (else
            (return 27.0 1.0)))
        (elif (Variables|Default|GetIsHovering)
          (return 0.0 2.0)
          (else
            (return 2e+01 1.0)))))))

(fn CalculateOffsetAndInterpSpeed
  (params (ViewRotation :type Rotator))
  (bind (_isinstate _iswindingup _winduppercent _divepercent) (Utilities|GetCrashingState))
  (LocalVariables|SetForwardOffset (* (Math|Vector|GetForwardVector (Variables|GetViewRotation)) (Variables|Default|GetCurrentForward)))
  (LocalVariables|SetRightOffset (* (Math|Vector|GetRightVector (Variables|GetViewRotation)) (Variables|Default|GetCurrentRight)))
  (LocalVariables|SetUpOffset (* "0.000000,0.000000,1.000000" (- (Variables|Default|GetCurrentUp) 14.0)))
  (LocalVariables|SetThrusting _isinstate)
  (LocalVariables|SetWindingUpThrust _iswindingup)
  (LocalVariables|SetThrustPercent _divepercent)
  (if (LocalVariables|GetThrusting)
    (if (LocalVariables|GetWindingUpThrust)
      (return (Utilities|Operators|Add (Utilities|Operators|Multiply (Math|Vector|GetRightVector (Variables|GetViewRotation)) 6e+01 (Variables|Default|GetSavedSign)) (* (Math|Vector|GetUpVector (Variables|GetViewRotation)) -5.0) (* (Math|Vector|GetForwardVector (Variables|GetViewRotation)) 3e+01)) 7.0)
      (else
        (return (+ (* (+ (LocalVariables|GetRightOffset) (LocalVariables|GetUpOffset)) (Math|Curves|GetFloatValue (Variables|Default|GetThrustingLateralAmountCurve) (LocalVariables|GetThrustPercent))) (* (LocalVariables|GetForwardOffset) (Math|Curves|GetFloatValue (Variables|Default|GetThrustingForwardAmountCurve) (LocalVariables|GetThrustPercent)))) (Utilities|RemapFloat (LocalVariables|GetThrustPercent) "(X=0.000000,Y=1.000000)" "(X=0.500000,Y=8.000000)" true))))
    (else
      (return (Utilities|Operators|Add (LocalVariables|GetForwardOffset) (LocalVariables|GetRightOffset) (LocalVariables|GetUpOffset)) 1e+01))))

(fn CalculateOffsetInterpSpeed
  (if (GameplayTags|HasMatchingGameplayTag (Variables|Default|GetPlayerPawn_Athena) "(TagName=\"Ability.HighTower.Power.Tomato.RepulsorCannon\")")
    (return 25.0)
    (elif (Variables|Default|GetIsHovering)
      (return 1e+01)
      (else
        (return 3.5)))))

(fn IsInLeadIn
  (bind (_isinstate _iswindingup) (Utilities|GetCrashingState))
  (return (Math|Boolean|ANDBoolean (< (- (Utilities|Time|GetGameTimeinSeconds) (Variables|Default|GetStartedTime)) (Variables|Default|GetLeadInTime)) (not _isinstate) (not _iswindingup))))

(fn SetStartingLocation
  (params (StartingLocation :type Vector))
  (Variables|Default|SetStartingLocation StartingLocation))

(fn UseAnchorPoint
  (params (ViewRotation :type Rotator) (AnchoredLocation :type Vector) (UpCurve :type Curve%20Float%20Object%20Reference) (ForwardCurve :type Curve%20Float%20Object%20Reference) (T :type Float%20%28double-precision%29) (UseActualForward :type Boolean))
  (bind _playerpawn_athena (Variables|Default|GetPlayerPawn_Athena))
  (bind _returnvalue (Transformation|GetActorLocation _playerpawn_athena))
  (bind _returnvalue_2 (- _returnvalue AnchoredLocation))
  (Variables|Default|SetAnchor_UpOffset (+ (* (Math|Float|Max(Float) (Math|Vector|DotProduct (Math|Vector|GetUpVector ViewRotation) _returnvalue_2)) (Math|Curves|GetFloatValue UpCurve T)) (Variables|Default|GetAnchor_UpOffset)))
  (Variables|Default|SetAnchor_ForwardOffset (+ (* (Math|Float|Max(Float) (Math|Float|SelectFloat (Math|Vector|DotProduct (Math|Vector|GetForwardVector ViewRotation) _returnvalue_2) (Math|Vector|VectorLength _returnvalue_2) (Variables|GetUseActualForward))) (Math|Curves|GetFloatValue ForwardCurve T)) (Variables|Default|GetAnchor_ForwardOffset)))
  (return))
