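; BarnTowel flying jetpack movement extension, linking flight states to thrust effects and repulsor attacks.
; Flying status changes toggle the wind-up GameplayEffect, apply fuel use and set flight intent.
; Crash action sends a cannon event if the pawn's weapon tag exists, else reports failure; shots send projectile events, drop flight intent and briefly block boost.
; Timed launch-back sets reversed horizontal plus upward velocity; the movement update only forwards the parent's result.
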(event Custom|Stop
  (Variables|Default|SetCurrentMaxSpeed_0 0.0))

(event Custom|LaunchBack
  (Variables|Default|SetCurrentMaxSpeed_0 999999.0)
  (Variables|Default|SetLaunchingBack true)
  (Variables|Default|SetLaunchingBackT 0.0)
  (Variables|Default|SetLaunchBackVelocity (* (+ (Math|Vector|NegateVector (Math|Vector|Normalize2D(Vector) (Variables|Default|GetCurrentUnmodifiedVelocity))) "0.000000,0.000000,1.000000") 2.6e+03))
  (Utilities|Time|SetTimerbyEvent (delegate Custom|StopLaunchingBack) 0.4))

(event Custom|EndBurst)

(event Custom|StopLaunchingBack
  (Variables|Default|SetLaunchingBack false))

(event Custom|CrashAction_Triggered_1
  (bind _ownerpawn (Variables|FortMovementMode_BaseExtLogic|GetOwnerPawn))
  (if (GameplayTags|HasMatchingGameplayTag _ownerpawn "(TagName=\"Granted.HighTower.Power.Tomato.Combo.InInventory\")")
    (Ability|SendGameplayEventtoActor _ownerpawn "(TagName=\"Event.Athena.BarnTowel.TriggerRepulsorCannon\")")
    (else
      (bind _asfort_player_pawn (Utilities|Casting|CastToFortPlayerPawn (Variables|FortMovementMode_BaseExtLogic|GetOwnerPawn)))
      (FailedAction|UnabletoPerformAction _asfort_player_pawn false (GameplayTags|MakeLiteralGameplayTagContainer "(GameplayTags=((TagName=\"Athena.Ability.Failed.BarnTowel.MissingWeapon\")))")) ; show reason for failing on screen if one is given
    )))

(event EventOnExtensionActivate
  (params (RuntimeContext :type MMERuntime%20Context%20Structure%20%28by%20ref%29))
  (|CallParentFunction RuntimeContext))

(event Custom|CheckIfTheresInput
  (if (and (> (Math|Vector2D|Vector2DLength (Variables|Default|GetCurrentMovementInput)) 0.5) (Variables|Default|GetInputShouldBoost))
    (Variables|FortMovementMode_ExtLogicFlyingJetpack|SetbIntendsToFly true)
    (Utilities|Time|InvalidateTimerHandle (Variables|Default|GetWaitingForInputTimer))))

(event EventHandleFlyingStatusChanged
  (params (MoveStatus :type Fort%20Movement%20Mode%20Jetpack%20Flying%20Runtime%20Data%20Object%20Reference) (NewStatus :type EBarnTowelFlyingStatus%20Enum :default "Default") (bIsActive :type Boolean) (FailureTag :type Gameplay%20Tag%20Structure%20%28by%20ref%29))
  (bind _ownerpawn (Variables|FortMovementMode_BaseExtLogic|GetOwnerPawn))
  (bind _returnvalue (Ability|GetFortAbilitySystemComponent _ownerpawn))
  (|CallParentFunction MoveStatus NewStatus bIsActive FailureTag)
  (switch Utilities|FlowControl|Switch|SwitchonEBarnTowelFlyingStatus NewStatus
    (:WindingUp
      (if bIsActive
        (GameplayEffects|ApplyGameplayEffectToSelf :self _returnvalue :GameplayEffectClass "/BarnTowelGameplay/Gameplay/GameplayEffects/GE_BarnTowel_WindingUpThrustFX.GE_BarnTowel_WindingUpThrustFX_C" :Level 0.0)
        (else
          (GameplayEffects|RemoveActiveEffectswithGrantedTags _returnvalue "(GameplayTags=((TagName=\"Granted.Athena.BarnTowel.WindingUpThrust\")))"))))
    (:Thrusting
      (if bIsActive
        (GameplayEffects|ApplyGameplayEffectToSelf :self (Ability|GetFortAbilitySystemComponent (Variables|FortMovementMode_BaseExtLogic|GetOwnerPawn)) :GameplayEffectClass "/BarnTowelGameplay/Gameplay/GameplayEffects/GE_UseFuel_Ram.GE_UseFuel_Ram_C" :Level 0.0)
        (Variables|FortMovementMode_ExtLogicFlyingJetpack|SetbIntendsToFly true)))))

(event Custom|AllowNextShot
  (Utilities|FlowControl|DoOnce false
    (:Completed
      (Ability|SendGameplayEventtoActor (Variables|FortMovementMode_BaseExtLogic|GetOwnerPawn) "(TagName=\"Event.Athena.BarnTowel.TriggerRepulsorProjectile\")")
      (Variables|FortMovementMode_ExtLogicFlyingJetpack|SetbIntendsToFly false)
      (Utilities|Time|SetTimerbyEvent (delegate Custom|AllowNextShot) 0.4)
      (Utilities|Time|InvalidateTimerHandle (Variables|Default|GetStartWaitingForInputTimer))
      (Variables|Default|SetInputShouldBoost false)
      (bind _returnvalue (Utilities|Time|SetTimerbyEvent (delegate Custom|StartWaitingForInput) 0.9))
      (Variables|Default|SetStartWaitingForInputTimer _returnvalue))))

(event Custom|StartWaitingForInput
  (Variables|Default|SetInputShouldBoost true))

(fn BlueprintUpdateForMovement
  (params (MoveStatus :type Fort%20Movement%20Mode%20Jetpack%20Flying%20Runtime%20Data%20Object%20Reference) (DeltaTime :type Float%20%28single-precision%29) (InVelocity :type Vector%20%28by%20ref%29 :default "0, 0, 0") (InAcceleration :type Vector%20%28by%20ref%29 :default "0, 0, 0"))
  (bind (_outnewvelocity _outnewacceleration) (|CallParentFunction MoveStatus DeltaTime InVelocity InAcceleration))
  (return _outnewvelocity _outnewacceleration))
