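; Parent movement-mode extension logic for BarnTowel jetpack flight: hover, ascend, thrust, dodge rolls and crash impacts.
; Activation binds input and weapon events and sets flight intents; status changes fire gameplay events, effects and cues, and deactivation resets state.
; Hit response skims along surfaces, crashes, applies low-friction landings, deflects or ends flight.
; Helper functions compute acceleration, goal yaw, hover targets and pitch noise from scalable-float tuning; a CVar flags native logic.
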(event Custom|ResetMoveInputValues
  ;[1 1 1] Reset Move Input Values
  (Variables|FortMovementMode_ExtLogicFlyingJetpack|SetbIntendsToFly false)
  (Variables|FortMovementMode_ExtLogicFlyingJetpack|SetbIntendsToAscend false)
  (Variables|FortMovementMode_ExtLogicFlyingJetpack|SetbIntendsToThrust false)
  (Variables|FortMovementMode_ExtLogicFlyingJetpack|SetbIntendsToRollLeft false)
  (Variables|FortMovementMode_ExtLogicFlyingJetpack|SetbIntendsToRollRight false))

(event Custom|OnUnequipped
  (Variables|FortMovementMode_ExtLogicFlyingJetpack|SetbIntendsToFly false)
  (bind _asplayer_pawn_athena (Utilities|Casting|CastToPlayerPawn_Athena (Variables|FortMovementMode_BaseExtLogic|GetOwnerPawn)))
  (Class|FortPlayerAnimationInterface|SetIsFlyingVertically (Animation|GetAnimationInterface _asplayer_pawn_athena) false)
  (Class|FortPlayerAnimationInterface|SetIsFlyingHorizontally (Animation|GetAnimationInterface _asplayer_pawn_athena) false))

(event Custom|OnEquipped
  (params (NewWeapon :type Fort%20Weapon%20Object%20Reference) (PrevWeapon :type Fort%20Weapon%20Object%20Reference))
  (Utilities|Casting|CastToB_Weapon_BarnTowel NewWeapon)
  (Variables|FortMovementMode_ExtLogicFlyingJetpack|SetbIntendsToFly true))

(event EventOnExtensionActivate
  (params (RuntimeContext :type MMERuntime%20Context%20Structure%20%28by%20ref%29))
  (bind _cachedplayercontroller (Variables|Default|GetCachedPlayerController))
  (bind _returnvalue (Actor|GetComponentbyClass _cachedplayercontroller "/Script/FortniteGame.FortControllerComponent_Minimap"))
  (bind _ownerpawn (Variables|FortMovementMode_BaseExtLogic|GetOwnerPawn))
  (bind _returnvalue_2 (Weapon|GetCurrentWeapon _ownerpawn))
  (bind (_asb_weapon_barn_towel _bsuccess) (Utilities|Casting|CastToB_Weapon_BarnTowel _returnvalue_2))
  (bind _ownerpawn_2 (Variables|FortMovementMode_BaseExtLogic|GetOwnerPawn))
  (Utilities|FlowControl|Sequence
    (:then_0
      (bind _returnvalue_3 (Development|GetConsoleVariableBoolValue "Flying.BarnTowel.UseNativeLogic"))
      (Variables|Default|SetCVar_UseNativeLogic _returnvalue_3))
    (:then_1
      (Development|PrintString (Variables|FortMovementMode_BaseExtLogic|GetOwnerPawn) (Utilities|String|Append (Utilities|GetDisplayName (Variables|FortMovementMode_BaseExtLogic|GetOwnerPawn)) " activated flying movement") true true "(R=1.000000,G=0.000000,B=0.825650,A=1.000000)" 5.0)
      (Variables|Default|SetCurrentMeshWorldRotation (Pawn|GetDesiredRotation (Pawn|GetController (Variables|FortMovementMode_BaseExtLogic|GetOwnerPawn))))
      (Variables|Default|SetMovementVector "0, 0, 0")
      (Variables|Default|SetLastMoveVecLengthWithInput 0.0)
      (Variables|Default|SetStartingVelocity (Pawn|Components|CharacterMovement|GetLastUpdateVelocity (Variables|/Script/Engine.Character|GetCharacterMovement (Variables|FortMovementMode_BaseExtLogic|GetOwnerPawn))))
      (Utilities|FlowControl|Sequence
        (:then_0
          (Variables|Default|SetTotalUnpaidCeilingHeightAdded 0.0)
          (Variables|Default|SetCeilingHeightToAddThisTick 0.0)
          (Variables|Default|SetPotentialCeilingHeightDecayThisTickFromFlying 0.0)
          (Variables|Default|SetActualCeilingHeightDecayThisTick 0.0)
          (Variables|Default|SetForwardSpeed (ScalableFloat|GetValueatLevel (Variables|Balance|Flying|GetLowerAccelerationFromHitPostProcessMovementVectorWeight)))
          (Variables|Default|SetTimeLastActivated (Actor|GetGameTimeSinceCreation (Variables|FortMovementMode_BaseExtLogic|GetOwnerPawn)))
          (Ability|Tasks|WaitGameplayTagAddToActor (Variables|FortMovementMode_BaseExtLogic|GetOwnerPawn) "(TagName=\"Granted.Athena.BarnTowel.ForceFall\")"
            (:then
              (if (not (Variables|Default|GetCVar_UseNativeLogic))
                (Ability|Async|WaitGameplayEventToActor (Variables|FortMovementMode_BaseExtLogic|GetOwnerPawn) "(TagName=\"Event.Athena.Flying.DoInstantThrust\")"
                  (:EventReceived
                    (CallFunction|SetBoost (ScalableFloat|GetValueatLevel (Variables|Balance|InstantThrust|GetInstantThrustBoost))) ; TODO NATIVIZE
                    (Variables|Default|SetBonusBoost (ScalableFloat|GetValueatLevel (Variables|Balance|InstantThrust|GetInstantThrustBonusBoost)))
                    (GameplayEffects|ApplyGameplayEffectToTarget :self (Ability|GetAbilitySystemComponent (Variables|FortMovementMode_BaseExtLogic|GetOwnerPawn)) :GameplayEffectClass (Variables|Smashing|GetSmashingGE) :Target (Ability|GetAbilitySystemComponent (Variables|FortMovementMode_BaseExtLogic|GetOwnerPawn)) :Level 0.0)))))
            (:Added
              (Variables|FortMovementMode_ExtLogicFlyingJetpack|SetbIntendsToFly false))))
        (:then_1
          (if (Utilities|Enum|Equal(Enum) (Networking|GetLocalRole _ownerpawn_2) "ROLE_AutonomousProxy")
            (bind _asplayer_controller (Utilities|Casting|CastToPlayerController (Pawn|GetController _ownerpawn_2)))
            (Variables|Default|SetCachedPlayerController _asplayer_controller)
            (CallFunction|ToggleInputMappingContext true)))
        (:then_2
          (bind _asfort_ability_system_component (Utilities|Casting|CastToFortAbilitySystemComponent (Ability|GetAbilitySystemComponent (Variables|FortMovementMode_BaseExtLogic|GetOwnerPawn))))
          (Variables|Default|SetOwnerASC _asfort_ability_system_component))
        (:then_3
          ;[0.15 0 0 0.5] Remove the following tags which seemed unused:
          ; - Athena.ContextualTutorial.SunRose.Wings.FlyForward.Start
          (Variables|FortMovementMode_ExtLogicFlyingJetpack|SetbIntendsToFly true)
          (bind _asplayer_pawn_athena (Utilities|Casting|CastToPlayerPawn_Athena (Variables|FortMovementMode_BaseExtLogic|GetOwnerPawn)))
          (FortPawn|BindEventtoOnWeaponUnEquippedDelegate _asplayer_pawn_athena (delegate Custom|OnUnequipped))
          (FortPawn|AssignOnWeaponEquippedDelegate _asplayer_pawn_athena (delegate Custom|OnEquipped)))
        (:then_4
          (bind _asynctaskproxy (Ability|Tasks|WaitGameplayTagRemoveFromActor (Variables|FortMovementMode_BaseExtLogic|GetOwnerPawn) (Variables|Default|GetTag_RequiredToContinueFlying))
            (:then
              (Variables|Default|SetTask_WaitRemoveTagRequiredToContinueFlying _asynctaskproxy))
            (:Removed
              (if _bsuccess
                (Class|FortWeaponJetpack|EndFlying _asb_weapon_barn_towel)))))
        (:then_5
          (bind _asfort_player_controller (Utilities|Casting|CastToFortPlayerController _cachedplayercontroller))
          (EventDispatchers|BindEventtoOnInventoryActiveChangedEvent _asfort_player_controller (EventDispatchers|CreateEvent :!event InputConditionChanged))
          (Utilities|IsValid _returnvalue
            (:Is%20Valid
              (PlayerController|BindEventtoOnToggleFullscreenMap _returnvalue (EventDispatchers|CreateEvent :!event InputConditionChanged)))))))))

(event Custom|OnWeaponEquippedDelegate_Event
  (params (NewWeapon :type Fort%20Weapon%20Object%20Reference) (PrevWeapon :type Fort%20Weapon%20Object%20Reference)))

(event Custom|UpdateFuelUsageRate
  (bind _returnvalue (Ability|GameplayEffect|AssignTagSetbyCallerMagnitude (GameplayEffects|MakeOutgoingSpec :self (Ability|GetFortAbilitySystemComponent (Variables|FortMovementMode_BaseExtLogic|GetOwnerPawn)) :GameplayEffectClass (Variables|Fuel|GetFuelUsageRateGE) :Level 0.0) (Variables|Fuel|GetTagUsageRateSetByCallerMagnitude) (Utilities|RemapFloat (Math|Vector|VectorLength (Variables|Default|GetLastFinalVelocity)) (Math|Vector2D|MakeVector2D (ScalableFloat|GetValueatLevel (Variables|Balance|Fuel|GetFuelUsageRateMagnitudeOldRangeX)) (ScalableFloat|GetValueatLevel (Variables|Balance|Fuel|GetFuelUsageRateMagnitudeOldRangeY))) (Math|Vector2D|MakeVector2D (ScalableFloat|GetValueatLevel (Variables|Balance|Fuel|GetFuelUsageRateMagnitudeNewRangeX)) (ScalableFloat|GetValueatLevel (Variables|Balance|Fuel|GetFuelUsageRateMagnitudeNewRangeY))) true)))
  (GameplayEffects|ApplyGameplayEffectSpecToSelf (Ability|GetFortAbilitySystemComponent (Variables|FortMovementMode_BaseExtLogic|GetOwnerPawn)) _returnvalue))

(event EnhancedInputActionBarnTowelFlyingQuickThrustAction
  (params (ActionValue :type Boolean) (ElapsedSeconds :type Float%20%28double-precision%29) (TriggeredSeconds :type Float%20%28double-precision%29) (InputAction :type Input%20Action%20Object%20Reference :default "/BarnTowelGameplay/Gameplay/Input/BarnTowelFlyingQuickThrustAction.BarnTowelFlyingQuickThrustAction")))

(event Custom|DoQuickThrust
  ;[1 1 1] Inputs
  (Variables|FortMovementMode_ExtLogicFlyingJetpack|SetbIntendsToWindUp false)
  (Variables|FortMovementMode_ExtLogicFlyingJetpack|SetbIntendsToThrust true)
  (Utilities|Time|InvalidateTimerHandle (Variables|Default|GetTimer_WaitingForMinWindUpTime)))

(event EnhancedInputActionBarnTowelFlyingThrustAction
  (params (ActionValue :type Boolean) (ElapsedSeconds :type Float%20%28double-precision%29) (TriggeredSeconds :type Float%20%28double-precision%29) (InputAction :type Input%20Action%20Object%20Reference :default "/BarnTowelGameplay/Gameplay/Input/BarnTowelFlyingThrustAction.BarnTowelFlyingThrustAction"))
  (bind _ownerpawn (Variables|FortMovementMode_BaseExtLogic|GetOwnerPawn))
  (bind _returnvalue (Actor|GetGameTimeSinceCreation _ownerpawn))
  (bind _prethrustwindupduration (Variables|Balance|DiveBomb|GetPreThrustWindUpDuration))
  (bind _returnvalue_2 (ScalableFloat|GetValueatLevel _prethrustwindupduration))
  (bind _timestartedwindingup (Variables|Default|GetTimeStartedWindingUp))
  (bind _returnvalue_3 (- _returnvalue _timestartedwindingup)))

(event EventOnExtensionDeactivate
  (params (RuntimeContext :type MMERuntime%20Context%20Structure%20%28by%20ref%29))
  (bind _cachedplayercontroller (Variables|Default|GetCachedPlayerController))
  (bind _returnvalue (Actor|GetComponentbyClass _cachedplayercontroller "/Script/FortniteGame.FortControllerComponent_Minimap"))
  (Development|PrintString (Variables|FortMovementMode_BaseExtLogic|GetOwnerPawn) (Utilities|String|Append (Utilities|GetDisplayName (Variables|FortMovementMode_BaseExtLogic|GetOwnerPawn)) " stopped flying movement"))
  (Utilities|FlowControl|Sequence
    (:then_0
      (Variables|Default|SetCrashEnabled false)
      (GameplayCue|RemoveGameplayCueOnActor(Looping) (Variables|FortMovementMode_BaseExtLogic|GetOwnerPawn) (Variables|GameplayCues|GetGC_EnergyDrain) (Ability|GameplayCue|MakeGameplayCueParameters 0.0))
      (Variables|Default|SetEnergyDrainEnabled false))
    (:then_1
      (CallFunction|ToggleInputMappingContext false)
      (bind _asfort_player_controller (Utilities|Casting|CastToFortPlayerController _cachedplayercontroller))
      (EventDispatchers|UnbindEventfromOnInventoryActiveChangedEvent _asfort_player_controller (EventDispatchers|CreateEvent :!event InputConditionChanged))
      (Utilities|IsValid _returnvalue
        (:Is%20Valid
          (PlayerController|UnbindEventfromOnToggleFullscreenMap _returnvalue (EventDispatchers|CreateEvent :!event InputConditionChanged)))))
    (:then_2
      (CallFunction|ResetMoveInputValues))
    (:then_3
      (Ability|Async|EndAction (Variables|Default|GetTask_WaitRemoveTagRequiredToContinueFlying)))
    (:then_4
      (Transformation|SetRelativeRotation (Variables|/Script/Engine.Character|GetMesh (Variables|FortMovementMode_BaseExtLogic|GetOwnerPawn)) "0.000000,-90.000000,0.000000" false false)
      (Variables|Default|SetIsBoosting false)
      (Variables|Default|SetBoost 0.0))))

(event Custom|ReachedMinWindUpTime
  ;[1 1 1] Inputs
  (Variables|FortMovementMode_ExtLogicFlyingJetpack|SetbIntendsToWindUp false)
  (Variables|FortMovementMode_ExtLogicFlyingJetpack|SetbIntendsToThrust true)
  (Utilities|Time|InvalidateTimerHandle (Variables|Default|GetTimer_WaitingForMinWindUpTime))
  (Variables|Default|SetTime%20Since%20Last%20Thrusted (Actor|GetGameTimeSinceCreation (Variables|FortMovementMode_BaseExtLogic|GetOwnerPawn))))

(event EnhancedInputActionBarnTowelFlyingAscendAction
  (params (ActionValue :type Boolean) (ElapsedSeconds :type Float%20%28double-precision%29) (TriggeredSeconds :type Float%20%28double-precision%29) (InputAction :type Input%20Action%20Object%20Reference :default "/BarnTowelGameplay/Gameplay/Input/BarnTowelFlyingAscendAction.BarnTowelFlyingAscendAction")))

(event EnhancedInputActionBarnTowelFlyingDescendAction
  (params (ActionValue :type Boolean) (ElapsedSeconds :type Float%20%28double-precision%29) (TriggeredSeconds :type Float%20%28double-precision%29) (InputAction :type Input%20Action%20Object%20Reference :default "/BarnTowelGameplay/Gameplay/Input/BarnTowelFlyingDescendAction.BarnTowelFlyingDescendAction")))

(event EnhancedInputActionBarnTowelFlyingUnibeamAction
  (params (ActionValue :type Boolean) (ElapsedSeconds :type Float%20%28double-precision%29) (TriggeredSeconds :type Float%20%28double-precision%29) (InputAction :type Input%20Action%20Object%20Reference :default "/BarnTowelGameplay/Gameplay/Input/BarnTowelFlyingUnibeamAction.BarnTowelFlyingUnibeamAction"))
  (bind _ownerpawn (Variables|FortMovementMode_BaseExtLogic|GetOwnerPawn)))

(event EnhancedInputActionBarnTowelFlyingRollLeftAction
  (params (ActionValue :type Boolean) (ElapsedSeconds :type Float%20%28double-precision%29) (TriggeredSeconds :type Float%20%28double-precision%29) (InputAction :type Input%20Action%20Object%20Reference :default "/BarnTowelGameplay/Gameplay/Input/BarnTowelFlyingRollLeftAction.BarnTowelFlyingRollLeftAction")))

(event EventHandleFlyingStatusChanged
  (params (MoveStatus :type Fort%20Movement%20Mode%20Jetpack%20Flying%20Runtime%20Data%20Object%20Reference) (NewStatus :type EBarnTowelFlyingStatus%20Enum :default "Default") (bIsActive :type Boolean) (FailureTag :type Gameplay%20Tag%20Structure%20%28by%20ref%29))
  (bind _ownerpawn (Variables|FortMovementMode_BaseExtLogic|GetOwnerPawn))
  (bind _returnvalue (Ability|GetAbilitySystemComponent _ownerpawn))
  (if (not (CallFunction|IsClientReplaying))
    (switch Utilities|FlowControl|Switch|SwitchonEBarnTowelFlyingStatus NewStatus
      (:Hovering
        (if bIsActive
          (Ability|SendGameplayEventtoActor (Variables|FortMovementMode_BaseExtLogic|GetOwnerPawn) "(TagName=\"Event.Athena.BarnTowel.StartedHovering\")")
          (Variables|Default|SetHovering true)
          (GameplayEffects|ApplyGameplayEffectToSelf :self _returnvalue :GameplayEffectClass "/BarnTowelGameplay/Gameplay/GameplayEffects/GE_BarnTowel_Hovering.GE_BarnTowel_Hovering_C" :Level 0.0)
          (else
            (Ability|SendGameplayEventtoActor (Variables|FortMovementMode_BaseExtLogic|GetOwnerPawn) "(TagName=\"Event.Athena.BarnTowel.EndedHovering\")")
            (Variables|Default|SetHovering false)
            (GameplayEffects|RemoveActiveEffectswithGrantedTags _returnvalue "(GameplayTags=((TagName=\"Granted.Athena.BarnTowel.Hovering\")))"))))
      (:WindingUp
        (if bIsActive
          (Ability|SendGameplayEventtoActor (Variables|FortMovementMode_BaseExtLogic|GetOwnerPawn) "(TagName=\"Event.Athena.BarnTowel.StartedWindingUpThrust\")")
          (GameplayEffects|ApplyGameplayEffectToSelf :self (Ability|GetAbilitySystemComponent (Variables|FortMovementMode_BaseExtLogic|GetOwnerPawn)) :GameplayEffectClass "/BarnTowelGameplay/Gameplay/GameplayEffects/GE_BarnTowel_BlockAttackActivations_FromThrusting.GE_BarnTowel_BlockAttackActivations_FromThrusting_C" :Level 0.0)))
      (:Thrusting
        ;[1 1 1] Crashing Status Change
        (if bIsActive
          (Ability|SendGameplayEventtoActor (Variables|FortMovementMode_BaseExtLogic|GetOwnerPawn) "(TagName=\"Event.Athena.BarnTowel.StartedThrusting\")")
          (GameplayEffects|ApplyGameplayEffectToSelf :self (Ability|GetAbilitySystemComponent (Variables|FortMovementMode_BaseExtLogic|GetOwnerPawn)) :GameplayEffectClass (Variables|Smashing|GetSmashingGE) :Level 0.0)))
      (:DodgingLeft
        ;[1 1 1] Dodging Left Status Change
        (if bIsActive
          (GameplayCue|ExecuteGameplayCue%28Burst%29Local (Variables|FortMovementMode_BaseExtLogic|GetOwnerPawn) (Variables|GameplayCues|GetGC_Dodge_Left) (Ability|GameplayCue|MakeGameplayCueParameters 0.0))))
      (:DodgingRight
        ;[1 1 1] Dodging Right Status Change
        (if bIsActive
          (GameplayCue|ExecuteGameplayCue%28Burst%29Local (Variables|FortMovementMode_BaseExtLogic|GetOwnerPawn) (Variables|GameplayCues|GetGC_Dodge_Right) (Ability|GameplayCue|MakeGameplayCueParameters 0.0)))))))

(event EnhancedInputActionBarnTowelFlyingRollRightAction
  (params (ActionValue :type Boolean) (ElapsedSeconds :type Float%20%28double-precision%29) (TriggeredSeconds :type Float%20%28double-precision%29) (InputAction :type Input%20Action%20Object%20Reference :default "/BarnTowelGameplay/Gameplay/Input/BarnTowelFlyingRollRightAction.BarnTowelFlyingRollRightAction")))

(event Custom|OnStartedBoosting)

(event Custom|OnFinishedBoosting
  (GameplayEffects|RemoveActiveEffectswithGrantedTags (Ability|GetFortAbilitySystemComponent (Variables|FortMovementMode_BaseExtLogic|GetOwnerPawn)) (Variables|FortMovementMode_ExtLogicFlyingJetpack|GetSmashingThingsTags)))

(event EventHandleFlyingStatusActivationFailed
  (params (FailedStatus :type EBarnTowelFlyingStatus%20Enum :default "Default") (FailureTag :type Gameplay%20Tag%20Structure%20%28by%20ref%29))
  ;[1 1 1] Handle flying status activation failed
  (Utilities|FlowControl|Sequence
    (:then_0
      (switch Utilities|FlowControl|Switch|SwitchonEBarnTowelFlyingStatus FailedStatus
        (:Descending
          (Variables|FortMovementMode_ExtLogicFlyingJetpack|SetbIntendsToDescend false))
        (:Ascending
          (Variables|FortMovementMode_ExtLogicFlyingJetpack|SetbIntendsToAscend false))
        (:WindingUp
          (Variables|FortMovementMode_ExtLogicFlyingJetpack|SetbIntendsToWindUp false))
        (:Thrusting
          (Variables|FortMovementMode_ExtLogicFlyingJetpack|SetbIntendsToThrust false))
        (:DodgingLeft
          (Variables|FortMovementMode_ExtLogicFlyingJetpack|SetbIntendsToRollLeft false))
        (:DodgingRight
          (Variables|FortMovementMode_ExtLogicFlyingJetpack|SetbIntendsToRollRight false))))
    (:then_1
      (if (GameplayTags|HasTag (Variables|Default|GetTags_ToAllowUnableText) FailureTag true)
        (if (ScalableFloat|GetValueasBool (Variables|Balance%28InBlueprint%29|GetSF_StartFlying_DisplayUnableText))
          (switch Utilities|FlowControl|Switch|SwitchonEBarnTowelFlyingStatus FailedStatus
            (:Flying
              _)
            (:Ascending
              _)
            (:Thrusting
              _)
            (:DodgingLeft
              _)
            (:DodgingRight
              _))
          (bind _asfort_player_pawn (Utilities|Casting|CastToFortPlayerPawn (Variables|FortMovementMode_BaseExtLogic|GetOwnerPawn)))
          (FailedAction|UnabletoPerformAction _asfort_player_pawn false (GameplayTags|MakeGameplayTagContainerfromTag FailureTag)))))))

(fn BlueprintUpdateForHitResponse
  (params (MoveStatus :type Fort%20Movement%20Mode%20Jetpack%20Flying%20Runtime%20Data%20Object%20Reference) (DeltaTime :type Float%20%28single-precision%29) (InVelocity :type Vector%20%28by%20ref%29 :default "0, 0, 0") (OldPosition :type Vector%20%28by%20ref%29 :default "0, 0, 0") (Adjustement :type Vector%20%28by%20ref%29 :default "0, 0, 0") (Hit :type Hit%20Result%20Structure%20%28by%20ref%29))
  (bind (_bblockinghit _binitialoverlap _time _distance _location _impactpoint _normal _impactnormal _physmat _hitactor) (Collision|BreakHitResult (Variables|GetHit)))
  (bind (_bblockinghit_2 _binitialoverlap_2 _time_2) (Collision|BreakHitResult (Variables|GetHit)))
  (bind (_bblockinghit_3 _binitialoverlap_3 _time_3 _distance_2 _location_2 _impactpoint_2 _normal_2 _impactnormal_2) (Collision|BreakHitResult (Variables|GetHit)))
  (bind (_bblockinghit_4 _binitialoverlap_4 _time_4 _distance_3 _location_3 _impactpoint_3 _normal_3 _impactnormal_3 _physmat_2 _hitactor_2 _hitcomponent _hitbonename _bonename _hititem _elementindex _faceindex _tracestart) (Collision|BreakHitResult (Variables|GetHit)))
  (bind _hit (Variables|GetHit))
  (bind (_bblockinghit_5 _binitialoverlap_5 _time_5 _distance_4 _location_4 _impactpoint_4 _normal_4 _impactnormal_4 _physmat_3 _hitactor_3) (Collision|BreakHitResult _hit))
  (bind (_bblockinghit_6 _binitialoverlap_6 _time_6 _distance_5 _location_5 _impactpoint_5 _normal_5 _impactnormal_5 _physmat_4 _hitactor_4) (Collision|BreakHitResult (Variables|GetHit)))
  (bind (_bblockinghit_7 _binitialoverlap_7 _time_7 _distance_6 _location_6 _impactpoint_6 _normal_6 _impactnormal_6 _physmat_5 _hitactor_5) (Collision|BreakHitResult (Variables|GetHit)))
  (bind _hit_2 (Variables|GetHit))
  (bind (_bblockinghit_8 _binitialoverlap_8 _time_8 _distance_7 _location_7 _impactpoint_7 _normal_7 _impactnormal_7) (Collision|BreakHitResult _hit_2))
  (bind _returnvalue (Math|Vector|DotProduct "0, 0, 0" _impactnormal_7))
  (bind _hit_3 (Variables|GetHit))
  (bind (_bblockinghit_9 _binitialoverlap_9 _time_9 _distance_8 _location_8 _impactpoint_8 _normal_8 _impactnormal_8 _physmat_6 _hitactor_6) (Collision|BreakHitResult _hit_3))
  (bind _movementvector (Variables|Default|GetMovementVector))
  (bind _returnvalue_2 (Math|Vector|NegateVector _movementvector))
  (bind _returnvalue_3 (Math|Vector|Normalize _returnvalue_2))
  (bind _returnvalue_4 (+ _returnvalue_3 _normal_4))
  (bind _returnvalue_5 (Math|Vector|Normalize2D(Vector) _returnvalue_4))
  (bind (_asfort_water_body_actor _bsuccess) (Utilities|Casting|CastToFortWaterBodyActor _hitactor_4))
  (bind (_asfort_water_body_actor_2 _bsuccess_2) (Utilities|Casting|CastToFortWaterBodyActor _hitactor_6))
  (bind _returnvalue_6 (* _returnvalue _impactnormal_7))
  (bind _lastfinalvelocity (Variables|Default|GetLastFinalVelocity))
  (bind _returnvalue_7 (- _lastfinalvelocity _returnvalue_6))
  (if (Collision|BreakHitResult (Variables|GetHit))
    (Variables|Default|SetSurfaceImpactDOTProduct (Math|Vector|DotProduct (Math|Vector|Normalize (Variables|Default|GetLastFinalVelocity)) _impactnormal_2))
    (Variables|Default|SetSpeedAlongSurface (Math|Vector|VectorLength _returnvalue_7))
    (if (Math|Boolean|ANDBoolean (> (Math|Trig|Cos(Degrees) (ScalableFloat|GetValueatLevel (Variables|Balance|Skimming|GetSkimmingMinRequiredAngleToSurface))) (Math|Float|Absolute(Float) (Variables|Default|GetSurfaceImpactDOTProduct))) (> (Variables|Default|GetSpeedAlongSurface) (ScalableFloat|GetValueatLevel (Variables|Balance|Skimming|GetSkimmingMinRequiredVelocityAlongSurface))) (> (Variables|/Script/BarnTowelRuntime.FortMovementMode_BarnTowelFlyingRuntimeData|GetCurrentHeightAboveWorld (Variables|GetMoveStatus)) (ScalableFloat|GetValueatLevel (Variables|Balance|Skimming|GetSkimmingMinRequiredHeightAboveWorld))) (not (Class|FortJetpackFlyingBlueprintLibrary|FlyingStatusHasAnyFlags (Variables|/Script/BarnTowelRuntime.FortMovementMode_BarnTowelFlyingRuntimeData|GetFlyingStatus (Variables|GetMoveStatus)) 32)))
      ;[1 1 1] Movement Along Surface
      (bind (_outhit _returnvalue_8) (Components|Movement|MoveUpdatedComponent (Variables|FortMovementMode_BaseExtLogic|GetOwnerMovementComponent) (* _returnvalue_7 (* (Variables|GetDeltaTime) (- 1.0 _time_2))) (Transformation|GetWorldRotation (Variables|/Script/Engine.MovementComponent|GetUpdatedComponent (Variables|FortMovementMode_BaseExtLogic|GetOwnerMovementComponent))) true false))
      (if _returnvalue_8
        (Development|PrintString (Variables|FortMovementMode_BaseExtLogic|GetOwnerPawn) (Utilities|String|Append " scrubbed against surface " (Utilities|GetDisplayName _hitactor)) true true "(R=1.000000,G=0.000000,B=0.825774,A=1.000000)")
        (return "Continue" false)
        (else
          _)))
    (Variables|Default|SetSurfaceImpactAngle (Math|Trig|Acos(Degrees) (* (Variables|Default|GetSurfaceImpactDOTProduct) -1.0))) ; Smaller angle means more direct impact
    (Variables|Default|SetCrashEnabled (and (not (CallFunction|IsinInitialHover (Variables|GetMoveStatus) (ScalableFloat|GetValueatLevel (Variables|Balance|HitResponse|GetHitResponseIsInInitialHoverBufferTime)))) (GameplayTags|HasAnyMatchingGameplayTags (Variables|FortMovementMode_BaseExtLogic|GetOwnerPawn) (Variables|FortMovementMode_ExtLogicFlyingJetpack|GetHitResponseCrashImpactTags))))
    (if (or (and (Variables|Default|GetCrashEnabled) (<= (Variables|Default|GetSurfaceImpactAngle) (ScalableFloat|GetValueatLevel (Variables|Balance|DiveBomb|GetMinAngleToPreventCrash)))) (Class|FortJetpackFlyingBlueprintLibrary|FlyingStatusHasAnyFlags (Variables|/Script/BarnTowelRuntime.FortMovementMode_BarnTowelFlyingRuntimeData|GetFlyingStatus (Variables|GetMoveStatus)) 32))
      (Utilities|FlowControl|Sequence
        (:then_0
          ;[1 1 1] Wait until crash windup finished
          (if (and (< (Class|FortJetpackFlyingBlueprintLibrary|GetFlyingStatusElapsedTime "Default" (Variables|GetMoveStatus)) (ScalableFloat|GetValueatLevel (Variables|Balance|DiveBomb|GetPreThrustWindUpDuration))) (Class|FortJetpackFlyingBlueprintLibrary|FlyingStatusHasAnyFlags (Variables|/Script/BarnTowelRuntime.FortMovementMode_BarnTowelFlyingRuntimeData|GetFlyingStatus (Variables|GetMoveStatus)) 32))
            (return "Continue" false)))
        (:then_1
          (LocalVariables|SetWantToSlide false)
          (Ability|SendGameplayEventtoActor (Variables|FortMovementMode_BaseExtLogic|GetOwnerPawn) (Variables|FortMovementMode_ExtLogicFlyingJetpack|GetEventTagToTriggerCrashImpact) (Utilities|Struct|MakeGameplayEventData "(TagName=\"\")" (Variables|FortMovementMode_BaseExtLogic|GetOwnerPawn) 0 0 0 "()" "(GameplayTags=)" "(GameplayTags=)" 0.0 (Ability|TargetData|AbilityTargetDatafromHitResult (Collision|MakeHitResult _bblockinghit_4 _binitialoverlap_4 _time_4 _distance_3 _location_3 _impactpoint_3 _normal_3 _impactnormal_3 _physmat_2 _hitactor_2 _hitcomponent _hitbonename _bonename _hititem _elementindex _faceindex _tracestart (Variables|GetInVelocity)))))
          (if (and (> (ScalableFloat|GetValueatLevel (Variables|Balance|CrashImpact|GetCrashImpactIgnoreWaterDuration)) 0) _bsuccess_2)
            (bind _asfort_player_pawn (Utilities|Casting|CastToFortPlayerPawn (Variables|FortMovementMode_BaseExtLogic|GetOwnerPawn))
              (:then
                (Collision|AddIgnoredActor _asfort_player_pawn _asfort_water_body_actor_2 (ScalableFloat|GetValueatLevel (Variables|Balance|CrashImpact|GetCrashImpactIgnoreWaterDuration))))
              (:CastFailed
                _)))))
      (else
        (LocalVariables|SetWantToSlide true)
        (if (GameplayTags|HasAnyMatchingGameplayTags (Variables|FortMovementMode_BaseExtLogic|GetOwnerPawn) (Variables|LowFriction|GetTagsToPreventLowFrictionLanding))
          _
          (elif (and (<= (Math|Trig|Acos(Degrees) (Math|Vector|DotProduct _normal_5 "0.000000,0.000000,1.000000")) (ScalableFloat|GetValueatLevel (Variables|Balance|LowFrictionLanding|GetMaxAngleForLowFrictionLanding))) (not _bsuccess))
            ;[1 1 1] Low Friction Landing
            (bind _returnvalue_9 (Ability|GameplayEffect|AssignTagSetbyCallerMagnitude (GameplayEffects|MakeOutgoingSpec :self (Ability|GetFortAbilitySystemComponent (Variables|FortMovementMode_BaseExtLogic|GetOwnerPawn)) :GameplayEffectClass (Variables|LowFriction|GetLowFrictionLandingGE) :Level 0.0) "(TagName=\"SetByCaller.Duration\")" (Math|Float|Clamp(Float) (Utilities|Operators|Multiply (Utilities|RemapFloat (Variables|Default|GetSpeedAlongSurface) (Math|Vector2D|MakeVector2D (ScalableFloat|GetValueatLevel (Variables|Balance|LowFrictionLanding|Speed|GetSpeedForMinLowFrictionMult)) (ScalableFloat|GetValueatLevel (Variables|Balance|LowFrictionLanding|Speed|GetSpeedForMaxLowFrictionMult))) (Math|Vector2D|MakeVector2D (ScalableFloat|GetValueatLevel (Variables|Balance|LowFrictionLanding|Speed|GetLowFrictionMultFromSpeedMin)) (ScalableFloat|GetValueatLevel (Variables|Balance|LowFrictionLanding|Speed|GetLowFrictionMultFromSpeedMax))) true) (Utilities|RemapFloat (Variables|Default|GetSurfaceImpactAngle) (Math|Vector2D|MakeVector2D (ScalableFloat|GetValueatLevel (Variables|Balance|LowFrictionLanding|Angle|GetImpactAngleForMaxLowFrictionDuration)) (ScalableFloat|GetValueatLevel (Variables|Balance|LowFrictionLanding|Angle|GetImpactAngleForMinLowFrictionDuration))) (Math|Vector2D|MakeVector2D (ScalableFloat|GetValueatLevel (Variables|Balance|LowFrictionLanding|Angle|GetLowFrictionMultFromImpactAngleMax)) (ScalableFloat|GetValueatLevel (Variables|Balance|LowFrictionLanding|Angle|GetLowFrictionMultFromImpactAngleMin))) true) (ScalableFloat|GetValueatLevel (Variables|Balance|LowFrictionLanding|GetLowFrictionDurationScalar))) (ScalableFloat|GetValueatLevel (Variables|Balance|LowFrictionLanding|GetLowFrictionDurationMin)) (ScalableFloat|GetValueatLevel (Variables|Balance|LowFrictionLanding|GetLowFrictionDurationMax)))))
            (GameplayEffects|ApplyGameplayEffectSpecToSelf (Ability|GetFortAbilitySystemComponent (Variables|FortMovementMode_BaseExtLogic|GetOwnerPawn)) _returnvalue_9)))
        (Variables|/Script/Engine.MovementComponent|SetVelocity (* (Math|Float|Min(Float) (ScalableFloat|GetValueatLevel (Variables|Balance|Bonk|GetVelocityClampOnBonk)) (Math|Vector|VectorLength (Variables|/Script/Engine.MovementComponent|GetVelocity (Variables|/Script/Engine.Character|GetCharacterMovement (Variables|FortMovementMode_BaseExtLogic|GetOwnerPawn))))) (Math|Vector|Normalize (Variables|/Script/Engine.MovementComponent|GetVelocity (Variables|/Script/Engine.Character|GetCharacterMovement (Variables|FortMovementMode_BaseExtLogic|GetOwnerPawn))))) (Variables|/Script/Engine.Character|GetCharacterMovement (Variables|FortMovementMode_BaseExtLogic|GetOwnerPawn)))))
    (Development|PrintString (Variables|FortMovementMode_BaseExtLogic|GetOwnerPawn) (Utilities|String|Append (Utilities|GetDisplayName (Variables|FortMovementMode_BaseExtLogic|GetOwnerPawn)) " just hit " (Utilities|GetDisplayName _hitactor_5) " end flying now") true true "(R=1.000000,G=0.000000,B=0.040099,A=1.000000)")
    ;[0 1 0.005593] New
    (if (and (or (Pawn|Components|CharacterMovement|IsWalkable (Variables|/Script/Engine.Character|GetCharacterMovement (Variables|FortMovementMode_BaseExtLogic|GetOwnerPawn)) (Variables|GetHit)) (GameplayTags|HasAnyMatchingGameplayTags (Variables|FortMovementMode_BaseExtLogic|GetOwnerPawn) (Variables|FortMovementMode_ExtLogicFlyingJetpack|GetForceCrashDamagedTags))) (not (Variables|/Script/BarnTowelRuntime.FortMovementMode_BarnTowelFlyingRuntimeData|GetbIsStartingFlying (Variables|GetMoveStatus))))
      (Development|PrintString 0 "END FLIGHT" true true "(R=1.000000,G=0.000000,B=0.946494,A=1.000000)" 1e+01)
      (return "AbortPhysUpdate" (LocalVariables|GetWantToSlide))
      (elif (and (Variables|/Script/Engine.Actor|GetbCanBeDamaged _hitactor_3) (CallFunction|IsBustingThroughThings (Variables|GetMoveStatus)))
        _
        (else
          (CallFunction|PlayHitFX)
          (bind _output_get (Variables|Default|SetMovementVector (* _returnvalue_5 (ScalableFloat|GetValueatLevel (Variables|Balance|HitResponse|GetHitResponseMovementVectorMultiplier)))))
          (Variables|Default|SetLastMoveVecLengthWithInput (Math|Vector|VectorLength _output_get)) ; Decelerate
          (Variables|Default|SetLoweredAccelerationFromHitSomething_CurrentTime (Variables|Default|GetLoweredAccelerationFromHitSomething_MaxTime))
          (Variables|/Script/FlyingCodeRuntime.FortMovementMode_FlyingRuntimeData|SetMoveOrientation (Math|Conversions|RotationFromXVector _returnvalue_5) (Variables|GetMoveStatus)))))
    (return "Continue" (LocalVariables|GetWantToSlide))
    (else
      (return "Continue" false))))

(fn CalculateAcceleration(whenthere'sinput)
  (params (Move%20Status :type Fort%20Movement%20Mode%20Jetpack%20Flying%20Runtime%20Data%20Object%20Reference))
  (bind _flyingstatus (Variables|/Script/BarnTowelRuntime.FortMovementMode_BarnTowelFlyingRuntimeData|GetFlyingStatus Move%20Status))
  (if (Class|FortJetpackFlyingBlueprintLibrary|FlyingStatusHasAnyFlags _flyingstatus 32)
    (return (ScalableFloat|GetValueatLevel (Variables|Balance|CalculateAcceleration|GetCalculateAccelerationReturnValueWhenThrusting)))
    (elif (Class|FortJetpackFlyingBlueprintLibrary|FlyingStatusHasAnyFlags _flyingstatus 8)
      (return (ScalableFloat|GetValueatLevel (Variables|Balance|CalculateAcceleration|GetCalculateAccelerationReturnValueWhenHovering)))
      (else
        (return (* (Utilities|RemapFloat (Math|Vector|VectorLength (Variables|Default|GetMovementVector)) (Math|Vector2D|MakeVector2D (ScalableFloat|GetValueatLevel (Variables|Balance|CalculateAcceleration|GetCalculateAccelerationMovementVectorLengthOldRangeX)) (ScalableFloat|GetValueatLevel (Variables|Balance|CalculateAcceleration|GetCalculateAccelerationMovementVectorLengthOldRangeY))) (Math|Vector2D|MakeVector2D (ScalableFloat|GetValueatLevel (Variables|Balance|CalculateAcceleration|GetCalculateAccelerationMovementVectorLengthNewRangeX)) (ScalableFloat|GetValueatLevel (Variables|Balance|CalculateAcceleration|GetCalculateAccelerationMovementVectorLengthNewRangeY))) true) (Utilities|RemapFloat (Variables|Default|GetLoweredAccelerationFromHitSomething_CurrentTime) (Math|Vector2D|MakeVector2D (ScalableFloat|GetValueatLevel (Variables|Balance|CalculateAcceleration|GetCalculateAccelerationHitSomethingTimeOldRangeX)) (ScalableFloat|GetValueatLevel (Variables|Balance|Orientation|GetNativeLoweredAccelerationFromHitSomethingMaxTime))) (Math|Vector2D|MakeVector2D (ScalableFloat|GetValueatLevel (Variables|Balance|CalculateAcceleration|GetCalculateAccelerationHitSomethingTimeNewRangeX)) (ScalableFloat|GetValueatLevel (Variables|Balance|CalculateAcceleration|GetCalculateAccelerationHitSomethingTimeNewRangeY))) true)))))))

(fn CalculateGoalYaw
  (params (Move%20Status :type Fort%20Movement%20Mode%20Jetpack%20Flying%20Runtime%20Data%20Object%20Reference))
  (return (.yaw (Math|Rotator|SelectRotator (Math|Conversions|RotationFromXVector (Math|Vector|RotateVector (Math|Conversions|ToVector(Vector2D) (Variables|Default|GetMovementInputWhenLastValid)) (Pawn|GetDesiredRotation (Pawn|GetController (Variables|FortMovementMode_BaseExtLogic|GetOwnerPawn))))) (Pawn|GetDesiredRotation (Pawn|GetController (Variables|FortMovementMode_BaseExtLogic|GetOwnerPawn))) (Math|Boolean|ANDBoolean (> (Math|Vector|VectorLength (Math|Conversions|ToVector(Vector2D) (Variables|Default|GetMovementInputWhenLastValid))) (ScalableFloat|GetValueatLevel (Variables|Balance|CalculateGoalYaw|GetCalculateGoalYawMovementInputThreshold))) (Utilities|Enum|NotEqual(Enum) (Variables|/Script/BarnTowelRuntime.FortMovementMode_BarnTowelFlyingRuntimeData|GetFlyingStatus Move%20Status) "Default") (or (CallFunction|ShouldDoSpeedChanges Move%20Status) (Class|FortJetpackFlyingBlueprintLibrary|FlyingStatusHasAnyFlags (Variables|/Script/BarnTowelRuntime.FortMovementMode_BarnTowelFlyingRuntimeData|GetFlyingStatus Move%20Status) 8)))))))

(fn CalculateInitialHoverXYTarget
  (params (MoveStatus :type Fort%20Movement%20Mode%20Jetpack%20Flying%20Runtime%20Data%20Object%20Reference))
  (bind (_returnvalue_x _returnvalue_y) (Math|Vector|BreakVector (* (Math|Vector|RotateVector (Math|Conversions|ToVector(Vector2D) (Variables|/Script/FlyingCodeRuntime.FortMovementMode_FlyingRuntimeData|GetMovementInput (Variables|GetMoveStatus))) (Pawn|GetDesiredRotation (Pawn|GetController (Variables|FortMovementMode_BaseExtLogic|GetOwnerPawn)))) (Utilities|RemapFloat (Class|FortJetpackFlyingBlueprintLibrary|GetFlyingStatusElapsedTime "Default" MoveStatus) (Math|Vector2D|MakeVector2D (* (Variables|Default|GetInitialHoverTime) (ScalableFloat|GetValueatLevel (Variables|Balance|CalculateInitialHoverTarget|GetInitialHoverTargetTimeMultipler))) (Variables|Default|GetInitialHoverTime)) (Math|Vector2D|MakeVector2D (Math|Float|SelectFloat 0.0 (ScalableFloat|GetValueatLevel (Variables|Balance|LaunchingMovement|GetLaunchingMovementBelowHoveringNewVelocityInterpTargetMaxSize)) (GameplayTags|HasMatchingGameplayTag (Variables|FortMovementMode_BaseExtLogic|GetOwnerPawn) (Variables|FortMovementMode_ExtLogicFlyingJetpack|GetJetpackInteriorTag))) (ScalableFloat|GetValueatLevel (Variables|Balance|CalculateInitialHoverTarget|GetInitialHoverTargetXYCalculationNewRangeY))) true))))
  (return _returnvalue_x _returnvalue_y))

(fn CalculateInitialHoverZTarget
  (params (TimeSinceStarted :type Float%20%28double-precision%29))
  (return (Utilities|RemapFloat TimeSinceStarted (Math|Vector2D|MakeVector2D 0.0 (* (ScalableFloat|GetValueatLevel (Variables|Balance|Hover|GetNativeInitialHoverTime)) (ScalableFloat|GetValueatLevel (Variables|Balance|LaunchingMovement|GetLaunchingMovementInitialHoverTimeMultipler)))) (Math|Vector2D|MakeVector2D (ScalableFloat|GetValueatLevel (Variables|Balance|LaunchingMovement|GetLaunchingMovementBelowHoveringFlyingDurationNewRangeX)) (ScalableFloat|GetValueatLevel (Variables|Balance|LaunchingMovement|GetLaunchingMovementBelowHoveringFlyingDurationNewRangeY))) true)))

(fn CalculatePitchNoise
  (params (MoveStatus :type Fort%20Movement%20Mode%20Jetpack%20Flying%20Runtime%20Data%20Object%20Reference))
  (if (Utilities|Enum|Equal(Enum) (Variables|/Script/BarnTowelRuntime.FortMovementMode_BarnTowelFlyingRuntimeData|GetFlyingStatus MoveStatus) "Default")
    (return 0.0)
    (else
      (return (Utilities|Operators|Multiply (Utilities|Operators|Add (* (Math|Trig|Sin(Radians) (* (Math|Trig|Cos(Radians) (* (* (Class|FortJetpackFlyingBlueprintLibrary|GetFlyingStatusElapsedTime "Default" MoveStatus) (Utilities|RemapFloat (Math|Vector|DotProduct (Math|Vector|GetForwardVector (Variables|/Script/FlyingCodeRuntime.FortMovementMode_FlyingRuntimeData|GetMoveOrientation MoveStatus)) (Math|Vector|GetForwardVector (Pawn|GetDesiredRotation (Pawn|GetController (Variables|FortMovementMode_BaseExtLogic|GetOwnerPawn))))) (Math|Vector2D|MakeVector2D (ScalableFloat|GetValueatLevel (Variables|Balance|CalculatePitchNoise|GetCalculatePitchNoiseFlyingTimeMultiplierOldRangeX)) (ScalableFloat|GetValueatLevel (Variables|Balance|CalculatePitchNoise|GetCalculatePitchNoiseFlyingTimeMultiplierOldRangeY))) (Math|Vector2D|MakeVector2D (ScalableFloat|GetValueatLevel (Variables|Balance|CalculatePitchNoise|GetCalculatePitchNoiseFlyingTimeMultiplierNewRangeX)) (ScalableFloat|GetValueatLevel (Variables|Balance|CalculatePitchNoise|GetCalculatePitchNoiseFlyingTimeMultiplierNewRangeY))) true)) 1.12)) 4.35)) 1.11) (* (Math|Trig|Cos(Radians) (* (Math|Trig|Sin(Radians) (* (* (Class|FortJetpackFlyingBlueprintLibrary|GetFlyingStatusElapsedTime "Default" MoveStatus) (Utilities|RemapFloat (Math|Vector|DotProduct (Math|Vector|GetForwardVector (Variables|/Script/FlyingCodeRuntime.FortMovementMode_FlyingRuntimeData|GetMoveOrientation MoveStatus)) (Math|Vector|GetForwardVector (Pawn|GetDesiredRotation (Pawn|GetController (Variables|FortMovementMode_BaseExtLogic|GetOwnerPawn))))) (Math|Vector2D|MakeVector2D (ScalableFloat|GetValueatLevel (Variables|Balance|CalculatePitchNoise|GetCalculatePitchNoiseFlyingTimeMultiplierOldRangeX)) (ScalableFloat|GetValueatLevel (Variables|Balance|CalculatePitchNoise|GetCalculatePitchNoiseFlyingTimeMultiplierOldRangeY))) (Math|Vector2D|MakeVector2D (ScalableFloat|GetValueatLevel (Variables|Balance|CalculatePitchNoise|GetCalculatePitchNoiseFlyingTimeMultiplierNewRangeX)) (ScalableFloat|GetValueatLevel (Variables|Balance|CalculatePitchNoise|GetCalculatePitchNoiseFlyingTimeMultiplierNewRangeY))) true)) 1.75)) 3.11)) 1.32) (* (Math|Trig|Sin(Radians) (* (* (Class|FortJetpackFlyingBlueprintLibrary|GetFlyingStatusElapsedTime "Default" MoveStatus) (Utilities|RemapFloat (Math|Vector|DotProduct (Math|Vector|GetForwardVector (Variables|/Script/FlyingCodeRuntime.FortMovementMode_FlyingRuntimeData|GetMoveOrientation MoveStatus)) (Math|Vector|GetForwardVector (Pawn|GetDesiredRotation (Pawn|GetController (Variables|FortMovementMode_BaseExtLogic|GetOwnerPawn))))) (Math|Vector2D|MakeVector2D (ScalableFloat|GetValueatLevel (Variables|Balance|CalculatePitchNoise|GetCalculatePitchNoiseFlyingTimeMultiplierOldRangeX)) (ScalableFloat|GetValueatLevel (Variables|Balance|CalculatePitchNoise|GetCalculatePitchNoiseFlyingTimeMultiplierOldRangeY))) (Math|Vector2D|MakeVector2D (ScalableFloat|GetValueatLevel (Variables|Balance|CalculatePitchNoise|GetCalculatePitchNoiseFlyingTimeMultiplierNewRangeX)) (ScalableFloat|GetValueatLevel (Variables|Balance|CalculatePitchNoise|GetCalculatePitchNoiseFlyingTimeMultiplierNewRangeY))) true)) 1.65)) 2.95) (* (Math|Trig|Cos(Radians) (* (* (Class|FortJetpackFlyingBlueprintLibrary|GetFlyingStatusElapsedTime "Default" MoveStatus) (Utilities|RemapFloat (Math|Vector|DotProduct (Math|Vector|GetForwardVector (Variables|/Script/FlyingCodeRuntime.FortMovementMode_FlyingRuntimeData|GetMoveOrientation MoveStatus)) (Math|Vector|GetForwardVector (Pawn|GetDesiredRotation (Pawn|GetController (Variables|FortMovementMode_BaseExtLogic|GetOwnerPawn))))) (Math|Vector2D|MakeVector2D (ScalableFloat|GetValueatLevel (Variables|Balance|CalculatePitchNoise|GetCalculatePitchNoiseFlyingTimeMultiplierOldRangeX)) (ScalableFloat|GetValueatLevel (Variables|Balance|CalculatePitchNoise|GetCalculatePitchNoiseFlyingTimeMultiplierOldRangeY))) (Math|Vector2D|MakeVector2D (ScalableFloat|GetValueatLevel (Variables|Balance|CalculatePitchNoise|GetCalculatePitchNoiseFlyingTimeMultiplierNewRangeX)) (ScalableFloat|GetValueatLevel (Variables|Balance|CalculatePitchNoise|GetCalculatePitchNoiseFlyingTimeMultiplierNewRangeY))) true)) 0.723)) 2.45) (* (Math|Trig|Sin(Radians) (* (Math|Trig|Cos(Radians) (* (* (Class|FortJetpackFlyingBlueprintLibrary|GetFlyingStatusElapsedTime "Default" MoveStatus) (Utilities|RemapFloat (Math|Vector|DotProduct (Math|Vector|GetForwardVector (Variables|/Script/FlyingCodeRuntime.FortMovementMode_FlyingRuntimeData|GetMoveOrientation MoveStatus)) (Math|Vector|GetForwardVector (Pawn|GetDesiredRotation (Pawn|GetController (Variables|FortMovementMode_BaseExtLogic|GetOwnerPawn))))) (Math|Vector2D|MakeVector2D (ScalableFloat|GetValueatLevel (Variables|Balance|CalculatePitchNoise|GetCalculatePitchNoiseFlyingTimeMultiplierOldRangeX)) (ScalableFloat|GetValueatLevel (Variables|Balance|CalculatePitchNoise|GetCalculatePitchNoiseFlyingTimeMultiplierOldRangeY))) (Math|Vector2D|MakeVector2D (ScalableFloat|GetValueatLevel (Variables|Balance|CalculatePitchNoise|GetCalculatePitchNoiseFlyingTimeMultiplierNewRangeX)) (ScalableFloat|GetValueatLevel (Variables|Balance|CalculatePitchNoise|GetCalculatePitchNoiseFlyingTimeMultiplierNewRangeY))) true)) 0.92)) 0.524)) 0.792)) (Utilities|RemapFloat (Math|Vector|DotProduct (Math|Vector|GetForwardVector (Variables|/Script/FlyingCodeRuntime.FortMovementMode_FlyingRuntimeData|GetMoveOrientation MoveStatus)) (Math|Vector|GetForwardVector (Pawn|GetDesiredRotation (Pawn|GetController (Variables|FortMovementMode_BaseExtLogic|GetOwnerPawn))))) (Math|Vector2D|MakeVector2D (ScalableFloat|GetValueatLevel (Variables|Balance|CalculatePitchNoise|GetCalculatePitchNoiseDotResultMultiplerOldRangeX)) (ScalableFloat|GetValueatLevel (Variables|Balance|CalculatePitchNoise|GetCalculatePitchNoiseDotResultMultiplerOldRangeY))) (Math|Vector2D|MakeVector2D (ScalableFloat|GetValueatLevel (Variables|Balance|CalculatePitchNoise|GetCalculatePitchNoiseDotResultMultiplerNewRangeX)) (ScalableFloat|GetValueatLevel (Variables|Balance|CalculatePitchNoise|GetCalculatePitchNoiseDotResultMultiplerNewRangeY))) true) (Utilities|RemapFloat (Math|Vector|VectorLength (Variables|Default|GetMovementVector)) (Math|Vector2D|MakeVector2D (ScalableFloat|GetValueatLevel (Variables|Balance|CalculatePitchNoise|GetCalculatePitchNoiseMovementVectorLengthMultiplierOldRangeX)) (ScalableFloat|GetValueatLevel (Variables|Balance|CalculatePitchNoise|GetCalculatePitchNoiseMovementVectorLengthMultiplierOldRangeY))) (Math|Vector2D|MakeVector2D (ScalableFloat|GetValueatLevel (Variables|Balance|CalculatePitchNoise|GetCalculatePitchNoiseMovementVectorLengthMultiplierNewRangeX)) (ScalableFloat|GetValueatLevel (Variables|Balance|CalculatePitchNoise|GetCalculatePitchNoiseMovementVectorLengthMultiplierNewRangeY))) true))))))

(fn CalculatePitchTurnSpeed
  ;[0 0.9588 1 0.5] TODO NATIVIZE -- Figured out that this function is basically always returning 1.1 so this simplifies things
  (return (ScalableFloat|GetValueatLevel (Variables|Balance%28inBlueprint%29|GetSF_PitchTurnSpeedWhenCoasting))))

(fn CalculateYawNoise
  (params (MoveStatus :type Fort%20Movement%20Mode%20Jetpack%20Flying%20Runtime%20Data%20Object%20Reference))
  (if (Utilities|Enum|Equal(Enum) (Variables|/Script/BarnTowelRuntime.FortMovementMode_BarnTowelFlyingRuntimeData|GetFlyingStatus MoveStatus) "Default")
    (return 0.0)
    (else
      (return (Utilities|Operators|Multiply (Utilities|Operators|Add (* (Math|Trig|Cos(Radians) (* (Math|Trig|Sin(Radians) (* (* (Class|FortJetpackFlyingBlueprintLibrary|GetFlyingStatusElapsedTime "Default" MoveStatus) (ScalableFloat|GetValueatLevel (Variables|Balance|CalculateYawNoise|GetCalculateYawNoiseFlyingTimeMultiplier))) 1.24)) 3.96)) 1.2) (* (Math|Trig|Sin(Radians) (* (Math|Trig|Cos(Radians) (* (* (Class|FortJetpackFlyingBlueprintLibrary|GetFlyingStatusElapsedTime "Default" MoveStatus) (ScalableFloat|GetValueatLevel (Variables|Balance|CalculateYawNoise|GetCalculateYawNoiseFlyingTimeMultiplier))) 1.59)) 3.25)) 1.55) (* (Math|Trig|Cos(Radians) (* (* (Class|FortJetpackFlyingBlueprintLibrary|GetFlyingStatusElapsedTime "Default" MoveStatus) (ScalableFloat|GetValueatLevel (Variables|Balance|CalculateYawNoise|GetCalculateYawNoiseFlyingTimeMultiplier))) 1.11)) 2.15) (* (Math|Trig|Sin(Radians) (* (* (Class|FortJetpackFlyingBlueprintLibrary|GetFlyingStatusElapsedTime "Default" MoveStatus) (ScalableFloat|GetValueatLevel (Variables|Balance|CalculateYawNoise|GetCalculateYawNoiseFlyingTimeMultiplier))) 0.6243)) 2.73) (* (Math|Trig|Cos(Radians) (* (Math|Trig|Sin(Radians) (* (* (Class|FortJetpackFlyingBlueprintLibrary|GetFlyingStatusElapsedTime "Default" MoveStatus) (ScalableFloat|GetValueatLevel (Variables|Balance|CalculateYawNoise|GetCalculateYawNoiseFlyingTimeMultiplier))) 1.1)) 0.524)) 0.715)) (Utilities|RemapFloat (Math|Float|Absolute(Float) (- (.yaw (Variables|/Script/FlyingCodeRuntime.FortMovementMode_FlyingRuntimeData|GetMoveOrientation MoveStatus)) (CallFunction|CalculateGoalYaw MoveStatus))) (Math|Vector2D|MakeVector2D (ScalableFloat|GetValueatLevel (Variables|Balance|CalculateYawNoise|GetCalculateYawNoiseMoveOrientationYawOldRangeX)) (ScalableFloat|GetValueatLevel (Variables|Balance|CalculateYawNoise|GetCalculateYawNoiseMoveOrientationYawOldRangeY))) (Math|Vector2D|MakeVector2D (ScalableFloat|GetValueatLevel (Variables|Balance|CalculateYawNoise|GetCalculateYawNoiseMoveOrientationYawNewRangeX)) (ScalableFloat|GetValueatLevel (Variables|Balance|CalculateYawNoise|GetCalculateYawNoiseMoveOrientationYawNewRangeY))) true) (Utilities|RemapFloat (Math|Vector|VectorLength (Variables|Default|GetMovementVector)) (Math|Vector2D|MakeVector2D (ScalableFloat|GetValueatLevel (Variables|Balance|CalculateYawNoise|GetCalculateYawNoiseMovementVectorLengthOldRangeX)) (ScalableFloat|GetValueatLevel (Variables|Balance|CalculateYawNoise|GetCalculateYawNoiseMovementVectorLengthOldRangeY))) (Math|Vector2D|MakeVector2D (ScalableFloat|GetValueatLevel (Variables|Balance|CalculateYawNoise|GetCalculateYawNoiseMovementVectorLengthNewRangeX)) (ScalableFloat|GetValueatLevel (Variables|Balance|CalculateYawNoise|GetCalculateYawNoiseMovementVectorLengthNewRangeY))) true))))))

(fn CanBeginFlying
  (return (GameplayTags|HasMatchingGameplayTag (Variables|FortMovementMode_BaseExtLogic|GetOwnerPawn) (Variables|Default|GetTag_RequiredToContinueFlying))))

(fn CanContinueFlyingStatus
  (params (MoveStatus :type Fort%20Movement%20Mode%20Jetpack%20Flying%20Runtime%20Data%20Object%20Reference) (StatusToContinue :type EBarnTowelFlyingStatus%20Enum :default "Default") (ElapsedDuration :type Float%20%28single-precision%29))
  (bind _movestatus (Variables|GetMoveStatus))
  (bind _movestatus_2 (Variables|GetMoveStatus))
  (bind _flyingstatus (Variables|/Script/BarnTowelRuntime.FortMovementMode_BarnTowelFlyingRuntimeData|GetFlyingStatus _movestatus_2))
  (bind _flyingstatus_2 (Variables|/Script/BarnTowelRuntime.FortMovementMode_BarnTowelFlyingRuntimeData|GetFlyingStatus _movestatus))
  (Utilities|IsValid MoveStatus
    (:Is%20Valid
      (switch Utilities|FlowControl|Switch|SwitchonEBarnTowelFlyingStatus StatusToContinue
        (:Default
          _)
        (:Descending
          ;[1 1 1] Ascend can continue check
          (if (Class|FortJetpackFlyingBlueprintLibrary|FlyingStatusHasAnyFlags _flyingstatus_2 32) :!label _label
            _
            (elif (Math|Boolean|ANDBoolean (Class|FortJetpackFlyingBlueprintLibrary|FlyingStatusHasAnyFlags _flyingstatus_2 2) (> (Class|FortJetpackFlyingBlueprintLibrary|GetFlyingStatusElapsedTime "Default" _movestatus) 0) (<= (Class|FortJetpackFlyingBlueprintLibrary|GetFlyingStatusElapsedTime "Default" _movestatus) (Variables|GetElapsedDuration)))
              _
              (else
                (return true 0))))
          (return false 0))
        (:Flying
          ;[1 1 1] Boost can continue check
          (if (Class|FortJetpackFlyingBlueprintLibrary|FlyingStatusHasAnyFlags _flyingstatus 32)
            _
            (elif (Math|Boolean|ANDBoolean (Class|FortJetpackFlyingBlueprintLibrary|FlyingStatusHasAnyFlags _flyingstatus 4) (> (Class|FortJetpackFlyingBlueprintLibrary|GetFlyingStatusElapsedTime "Default" _movestatus_2) 0) (<= (Class|FortJetpackFlyingBlueprintLibrary|GetFlyingStatusElapsedTime "Default" _movestatus_2) (Variables|GetElapsedDuration)))
              _
              (else
                (return true 0))))
          (return false 0))
        (:Ascending
          (goto _label))
        (:Thrusting
          _)
        (:DodgingLeft
          ;[1 1 1] Dodge can continue check
          (return (not (Class|FortJetpackFlyingBlueprintLibrary|FlyingStatusHasAnyFlags (Variables|/Script/BarnTowelRuntime.FortMovementMode_BarnTowelFlyingRuntimeData|GetFlyingStatus (Variables|GetMoveStatus)) 32)) 0))
        (:DodgingRight
          (return (not (Class|FortJetpackFlyingBlueprintLibrary|FlyingStatusHasAnyFlags (Variables|/Script/BarnTowelRuntime.FortMovementMode_BarnTowelFlyingRuntimeData|GetFlyingStatus (Variables|GetMoveStatus)) 32)) 0)))
      (return true 0))))

(fn CanEndFlyingStatus
  (params (MoveStatus :type Fort%20Movement%20Mode%20Jetpack%20Flying%20Runtime%20Data%20Object%20Reference) (StatusToEnd :type EBarnTowelFlyingStatus%20Enum :default "Default") (ElapsedDuration :type Float%20%28single-precision%29))
  (Utilities|IsValid MoveStatus
    (:Is%20Valid
      (LocalVariables|SetIsCrashing (Class|FortJetpackFlyingBlueprintLibrary|FlyingStatusHasAnyFlags (Variables|/Script/BarnTowelRuntime.FortMovementMode_BarnTowelFlyingRuntimeData|GetFlyingStatus MoveStatus) 32))
      (switch Utilities|FlowControl|Switch|SwitchonEBarnTowelFlyingStatus StatusToEnd
        (:Default
          _)
        (:Descending
          (return (or (LocalVariables|GetIsCrashing) (and (>= (Variables|GetElapsedDuration) (ScalableFloat|GetValueatLevel (Variables|Balance%28InBlueprint%29|GetSF_Ascend_MinDuration))) (not (Variables|FortMovementMode_ExtLogicFlyingJetpack|GetbIntendsToDescend))))))
        (:Flying
          ;[1 1 1] Boost End Condition
          (return true))
        (:Ascending
          ;[1 1 1] Ascend End Condition
          (return (or (LocalVariables|GetIsCrashing) (and (>= (Variables|GetElapsedDuration) (ScalableFloat|GetValueatLevel (Variables|Balance%28InBlueprint%29|GetSF_Ascend_MinDuration))) (not (Variables|FortMovementMode_ExtLogicFlyingJetpack|GetbIntendsToAscend))))))
        (:Hovering
          (return (not (Variables|FortMovementMode_ExtLogicFlyingJetpack|GetbIntendsToHover))))
        (:WindingUp
          (if (Variables|FortMovementMode_ExtLogicFlyingJetpack|GetbIntendsToWindUp)
            (return false)
            (else
              (return true))))
        (:Thrusting
          _)
        (:DodgingLeft
          ;[1 1 1] Dodge End condition
          (return (or (LocalVariables|GetIsCrashing) (>= (Variables|GetElapsedDuration) (ScalableFloat|GetValueatLevel (Variables|Balance|Dodge|GetDodgeDuration))))))
        (:DodgingRight
          (return (or (LocalVariables|GetIsCrashing) (>= (Variables|GetElapsedDuration) (ScalableFloat|GetValueatLevel (Variables|Balance|Dodge|GetDodgeDuration)))))))
      (return false))
    (:Is%20Not%20Valid
      (return false))))

(fn CanStartFlyingStatus
  (params (MoveStatus :type Fort%20Movement%20Mode%20Jetpack%20Flying%20Runtime%20Data%20Object%20Reference) (StatusToStart :type EBarnTowelFlyingStatus%20Enum :default "Default"))
  (bind _movestatus (Variables|GetMoveStatus))
  (Utilities|IsValid MoveStatus
    (:Is%20Valid
      (switch Utilities|FlowControl|Switch|SwitchonEBarnTowelFlyingStatus StatusToStart
        (:Default
          _)
        (:Descending
          (if (or (Variables|/Script/BarnTowelRuntime.FortMovementMode_BarnTowelFlyingRuntimeData|GetbIsStartingFlying _movestatus) (GameplayTags|HasMatchingGameplayTag (Variables|FortMovementMode_BaseExtLogic|GetOwnerPawn) (Variables|FortMovementMode_ExtLogicFlyingJetpack|GetRepulsorCannonTag))) :!label _label_2
            (return false "(TagName=\"Athena.Ability.Failed.CannotDoThatNow\")")
            (else
              (if (Class|FortJetpackFlyingBlueprintLibrary|FlyingStatusHasAnyFlags (Variables|/Script/BarnTowelRuntime.FortMovementMode_BarnTowelFlyingRuntimeData|GetFlyingStatus _movestatus) 48) :!label _label
                (return false "(TagName=\"\")")
                (else
                  (return true "(TagName=\"\")"))))))
        (:Flying
          (goto _label))
        (:Ascending
          (goto _label_2))
        (:Hovering
          _)
        (:WindingUp
          _)
        (:Thrusting
          _)
        (:DodgingLeft
          (goto _label_2))
        (:DodgingRight
          (goto _label_2))))
    (:Is%20Not%20Valid
      _))
  (return true 0))

(fn CanStartWindingUpThrust
  (return (Math|Boolean|ANDBoolean (not (Utilities|Time|IsValidTimerHandle (Variables|Default|GetTimer_WaitingForMinWindUpTime))) (not (Variables|FortMovementMode_ExtLogicFlyingJetpack|GetbIntendsToWindUp)) (not (GameplayTags|HasMatchingGameplayTag (Variables|FortMovementMode_BaseExtLogic|GetOwnerPawn) "(TagName=\"Granted.Athena.BarnTowel.BlockAttackActivations\")")) (> (- (Actor|GetGameTimeSinceCreation (Variables|FortMovementMode_BaseExtLogic|GetOwnerPawn)) (Variables|Default|GetTimeLastActivated)) (+ (Variables|Default|GetInitialHoverTime) 0.5)) (> (- (Actor|GetGameTimeSinceCreation (Variables|FortMovementMode_BaseExtLogic|GetOwnerPawn)) (Variables|Default|GetTime%20Since%20Last%20Thrusted)) 0.7) (not (GameplayTags|HasMatchingGameplayTag (Variables|FortMovementMode_BaseExtLogic|GetOwnerPawn) "(TagName=\"Granted.Athena.BarnTowel.ForceFall\")")) (not (GameplayTags|HasMatchingGameplayTag (Variables|FortMovementMode_BaseExtLogic|GetOwnerPawn) (Variables|FortMovementMode_ExtLogicFlyingJetpack|GetIsInChargingStateRepulsorCannonTag))) (or (not (TouchPlatform|IsUsingTouch (Fortnite|GetFortPlayerControllerfromActor (Variables|FortMovementMode_BaseExtLogic|GetOwnerPawn)))) (GameplayTags|HasAnyMatchingGameplayTags (Variables|FortMovementMode_BaseExtLogic|GetOwnerPawn) (GameplayTags|MakeGameplayTagContainerfromArray (Utilities|Array|MakeArray "(TagName=\"Granted.Athena.BarnTowel.DoingTakeOff\")" "(TagName=\"Granted.Athena.BarnTowel.Flying\")" "(TagName=\"Granted.Athena.BarnTowel.WindingUpThrust\")")))))))

(fn ClampToMaxHeight
  (params (InMoveVec :type Vector) (MoveStatus :type Fort%20Movement%20Mode%20Jetpack%20Flying%20Runtime%20Data%20Object%20Reference))
  (bind _ownerpawn (Variables|FortMovementMode_BaseExtLogic|GetOwnerPawn))
  (bind (_returnvalue_x _returnvalue_y _returnvalue_z) (Math|Vector|BreakVector (Transformation|GetActorLocation _ownerpawn)))
  (bind _heightforminaddedceilingdecay (Variables|Balance|CeilingHeight|HeightAboveGround|GetHeightForMinAddedCeilingDecay))
  (bind _returnvalue (ScalableFloat|GetValueatLevel _heightforminaddedceilingdecay))
  (bind _globalceilingheightmax (Variables|Balance|CeilingHeight|GlobalHeight|GetGlobalCeilingHeightMax))
  (bind _returnvalue_2 (ScalableFloat|GetValueatLevel _globalceilingheightmax))
  (bind _clamptomaxheightceilingheightmaxmodifier (Variables|Balance|ClampToMaxHeight|GetClampToMaxHeightCeilingHeightMaxModifier))
  (bind _returnvalue_3 (ScalableFloat|GetValueatLevel _clamptomaxheightceilingheightmaxmodifier))
  (bind _returnvalue_4 (- _returnvalue_2 _returnvalue_3))
  (bind _interpolatedheightaboveworld (Variables|/Script/BarnTowelRuntime.FortMovementMode_BarnTowelFlyingRuntimeData|GetInterpolatedHeightAboveWorld MoveStatus))
  (if (> _returnvalue_z _returnvalue_4)
    (return (Math|Vector|MakeVector (.x (Variables|GetInMoveVec)) (.y (Variables|GetInMoveVec)) (Math|Float|Min(Float) (.z (Variables|GetInMoveVec)) (Utilities|RemapFloat _returnvalue_z (Math|Vector2D|MakeVector2D _returnvalue_4 _returnvalue_2) (Math|Vector2D|MakeVector2D (ScalableFloat|GetValueatLevel (Variables|Balance|ClampToMaxHeight|GetClampToMaxHeightPawnLocationZNewRangeX)) (ScalableFloat|GetValueatLevel (Variables|Balance|ClampToMaxHeight|GetClampToMaxHeightPawnLocationZNewRangeY))) true))))
    (elif (> _interpolatedheightaboveworld _returnvalue)
      (return (Math|Vector|MakeVector (.x (Variables|GetInMoveVec)) (.y (Variables|GetInMoveVec)) (Math|Float|Min(Float) (.z (Variables|GetInMoveVec)) (Utilities|RemapFloat _interpolatedheightaboveworld (Math|Vector2D|MakeVector2D _returnvalue (ScalableFloat|GetValueatLevel (Variables|Balance|CeilingHeight|HeightAboveGround|GetHeightForMaxAddedCeilingDecay))) (Math|Vector2D|MakeVector2D (ScalableFloat|GetValueatLevel (Variables|Balance|ClampToMaxHeight|GetClampToMaxHeightAboveWorldNewRangeX)) (ScalableFloat|GetValueatLevel (Variables|Balance|ClampToMaxHeight|GetClampToMaxHeightAboveWorldNewRangeY))) true))))
      (else
        (return (Variables|GetInMoveVec))))))

(fn ConditionsMetToEnterHover
  (params (Move%20Status :type Fort%20Movement%20Mode%20Jetpack%20Flying%20Runtime%20Data%20Object%20Reference))
  (bind _movementinput_x (Math|Vector2D|BreakVector2D (Variables|/Script/FlyingCodeRuntime.FortMovementMode_FlyingRuntimeData|GetMovementInput_X Move%20Status)))
  (bind _movementinput_x_2 (Math|Vector2D|BreakVector2D (Variables|/Script/FlyingCodeRuntime.FortMovementMode_FlyingRuntimeData|GetMovementInput_X Move%20Status)))
  (return (and (Math|Boolean|ORBoolean (and (< (Math|Vector|VectorLength (Variables|Default|GetMovementVector)) (ScalableFloat|GetValueatLevel (Variables|Balance|EnterHover|GetEnterHoverMovementVectorThreshold))) (< (Math|Vector2D|Vector2DLength (Variables|/Script/FlyingCodeRuntime.FortMovementMode_FlyingRuntimeData|GetMovementInput Move%20Status)) (ScalableFloat|GetValueatLevel (Variables|Balance|EnterHover|GetEnterHoverMovementInputLengthThreshold)))) (< _movementinput_x_2 (ScalableFloat|GetValueatLevel (Variables|Balance|EnterHover|GetEnterHoverMovementInputXBackwardThreshold))) (GameplayTags|HasMatchingGameplayTag (Variables|FortMovementMode_BaseExtLogic|GetOwnerPawn) (Variables|FortMovementMode_ExtLogicFlyingJetpack|GetRepulsorCannonTag)) (and (< _movementinput_x (ScalableFloat|GetValueatLevel (Variables|Balance|EnterHover|GetEnterHoverMovementInputXThreshold))) (< (CallFunction|GetInputAgreementwithMovement Move%20Status) (ScalableFloat|GetValueatLevel (Variables|Balance|EnterHover|GetEnterHoverInputAgreementWithMovementThreshold)))) (> (Variables|Default|GetLoweredAccelerationFromHitSomething_CurrentTime) 0)) (not (Class|FortJetpackFlyingBlueprintLibrary|FlyingStatusHasAnyFlags (Variables|/Script/BarnTowelRuntime.FortMovementMode_BarnTowelFlyingRuntimeData|GetFlyingStatus Move%20Status) 32)))))

(fn ConditionsMetToExitHover
  (params (Move%20Status :type Fort%20Movement%20Mode%20Jetpack%20Flying%20Runtime%20Data%20Object%20Reference))
  (bind _movementinput_x (Math|Vector2D|BreakVector2D (Variables|/Script/FlyingCodeRuntime.FortMovementMode_FlyingRuntimeData|GetMovementInput_X Move%20Status)))
  (bind (_movementinput_x_2 _movementinput_y) (Math|Vector2D|BreakVector2D (Variables|/Script/FlyingCodeRuntime.FortMovementMode_FlyingRuntimeData|GetMovementInput_X Move%20Status)))
  (return (and (Math|Boolean|ORBoolean (Math|Boolean|ANDBoolean (> _movementinput_x_2 (ScalableFloat|GetValueatLevel (Variables|Balance|ExitHover|GetExitHoverMovementInputXBackwardThreshold))) (> (Math|Float|Absolute(Float) _movementinput_y) (ScalableFloat|GetValueatLevel (Variables|Balance|ExitHover|GetExitHoverABSMovementInputYThreshold))) (> (Math|Vector|VectorLength (Variables|Default|GetMovementVector)) (ScalableFloat|GetValueatLevel (Variables|Balance|ExitHover|GetExitHoverMovementVectorLengthThreshold))) (not (GameplayTags|HasMatchingGameplayTag (Variables|FortMovementMode_BaseExtLogic|GetOwnerPawn) (Variables|FortMovementMode_ExtLogicFlyingJetpack|GetRepulsorCannonTag))) (> (CallFunction|GetInputAgreementwithMovement Move%20Status) (ScalableFloat|GetValueatLevel (Variables|Balance|ExitHover|GetExitHoverInputAgreementWithMovementThreshold)))) (> _movementinput_x (ScalableFloat|GetValueatLevel (Variables|Balance|ExitHover|GetExitHoverMovementInputXThreshold))) (Class|FortJetpackFlyingBlueprintLibrary|FlyingStatusHasAnyFlags (Variables|/Script/BarnTowelRuntime.FortMovementMode_BarnTowelFlyingRuntimeData|GetFlyingStatus Move%20Status) 32)) (> (.x (CallFunction|GetMovementVectorRelativetoCamera)) (ScalableFloat|GetValueatLevel (Variables|Balance|ExitHover|GetExitHoverCameraRelativeMovementVectorXThreshold))))))

(fn CurrentDistanceFromCeilingHeight
  (params (Move%20Status :type Fort%20Movement%20Mode%20Flying%20Runtime%20Data%20Object%20Reference))
  (bind (_returnvalue_x _returnvalue_y _returnvalue_z) (Math|Vector|BreakVector (Transformation|GetActorLocation (Variables|FortMovementMode_BaseExtLogic|GetOwnerPawn))))
  (return (Math|Float|Max(Float) (- (Variables|/Script/FlyingCodeRuntime.FortMovementMode_FlyingRuntimeData|GetCeilingHeight Move%20Status) _returnvalue_z))))

(fn GetInputAgreementWithMovement
  (params (MoveStatus :type Fort%20Movement%20Mode%20Flying%20Runtime%20Data%20Object%20Reference))
  (return (Math|Vector2D|DotProduct(2D) (Variables|/Script/FlyingCodeRuntime.FortMovementMode_FlyingRuntimeData|GetMovementInput MoveStatus) (Math|Conversions|ToVector2D(Vector) (Math|Vector|VectorClampSizeMax2D (CallFunction|GetMovementVectorRelativetoCamera) 1.0)))))

(fn GetMovementVectorRelativeToCamera
  ;[0 0.91 1 0.5] TODO NATIVIZE -- New function
  (return (Math|Vector|RotateVector (Variables|Default|GetMovementVector) (Math|Rotator|InvertRotator (Pawn|GetDesiredRotation (Pawn|GetController (Variables|FortMovementMode_BaseExtLogic|GetOwnerPawn)))))))

(fn InputConditionChanged
  (params (InventoryActive :type Boolean))
  (CallFunction|ToggleInputMappingContext (not InventoryActive)))

(fn InputMultiplierOnRotation
  (params (Move%20Status :type Fort%20Movement%20Mode%20Jetpack%20Flying%20Runtime%20Data%20Object%20Reference) (Is%20Pitch :type Boolean))
  (bind (_result _shouldrotatetoo) (CallFunction|TEMPIsinChargingState Move%20Status))
  (return (Math|Float|SelectFloat (ScalableFloat|GetValueatLevel (Variables|Balance|InputMultiplierOnRotation|GetInputMultiplierOnRotationChargingStateMultiplier)) (Math|Float|SelectFloat (Math|Float|SelectFloat 1.0 (Math|Float|Max(Float) (Math|Vector2D|BreakVector2D (Variables|/Script/FlyingCodeRuntime.FortMovementMode_FlyingRuntimeData|GetMovementInput Move%20Status)) (ScalableFloat|GetValueatLevel (Variables|Balance|InputMultiplierOnRotation|GetInputMultiplierOnRotationMovementInputMax))) (or (Variables|/Script/BarnTowelRuntime.FortMovementMode_BarnTowelFlyingRuntimeData|GetbIsStartingFlying Move%20Status) (>= (Variables|Default|GetBoost) 1.0))) (Math|Float|SelectFloat 0.0 (ScalableFloat|GetValueatLevel (Variables|Balance|InputMultiplierOnRotation|GetInputMultiplierOnRotationNotHoveringFactor)) (Utilities|IsHovering Move%20Status)) Is%20Pitch) (and _result _shouldrotatetoo))))

(fn IsBustingThroughThings
  (params (MoveStatus :type Fort%20Movement%20Mode%20Jetpack%20Flying%20Runtime%20Data%20Object%20Reference))
  (return (or (Class|FortJetpackFlyingBlueprintLibrary|FlyingStatusHasAnyFlags (Variables|/Script/BarnTowelRuntime.FortMovementMode_BarnTowelFlyingRuntimeData|GetFlyingStatus MoveStatus) 4) (Variables|/Script/BarnTowelRuntime.FortMovementMode_BarnTowelFlyingRuntimeData|GetbIsStartingFlying MoveStatus))))

(fn IsInInitialHover
  (params (MoveStatus :type Fort%20Movement%20Mode%20Jetpack%20Flying%20Runtime%20Data%20Object%20Reference) (BufferTime :type Float%20%28double-precision%29))
  (return (and (< (Class|FortJetpackFlyingBlueprintLibrary|GetFlyingStatusElapsedTime "Default" MoveStatus) (+ (ScalableFloat|GetValueatLevel (Variables|Balance|Hover|GetNativeInitialHoverTime)) BufferTime)) (Variables|/Script/BarnTowelRuntime.FortMovementMode_BarnTowelFlyingRuntimeData|GetbIsStartingFlying MoveStatus))))

(fn IsInPreCrashWindUp
  (params (MoveStatus :type Fort%20Movement%20Mode%20Barn%20Towel%20Flying%20Runtime%20Data%20Object%20Reference))
  (return (Class|FortJetpackFlyingBlueprintLibrary|FlyingStatusHasAnyFlags (Variables|/Script/BarnTowelRuntime.FortMovementMode_BarnTowelFlyingRuntimeData|GetFlyingStatus MoveStatus) 16)))

(fn PlayHitFX
  (params (Hit :type Hit%20Result%20Structure%20%28by%20ref%29))
  (bind (_bblockinghit _binitialoverlap _time _distance _location _impactpoint _normal _impactnormal _physmat _hitactor _hitcomponent) (Collision|BreakHitResult Hit))
  (bind _ownerpawn (Variables|FortMovementMode_BaseExtLogic|GetOwnerPawn))
  (bind _currentweapon (Variables|/Script/FortniteGame.FortPawn|GetCurrentWeapon _ownerpawn))
  (bind (_asb_weapon_barn_towel _bsuccess) (Utilities|Casting|CastToB_Weapon_BarnTowel _currentweapon))
  (bind (_asfort_water_body_actor _bsuccess_2) (Utilities|Casting|CastToFortWaterBodyActor _hitactor))
  (if (and (> (Math|Vector|VectorLength (Variables|Default|GetMovementVector)) (ScalableFloat|GetValueatLevel (Variables|Balance|Fuel|GetPlayHitFXMovementVectorLengthThreshold))) (not _bsuccess_2))
    (if _bsuccess
      (GameplayCue|ExecuteGameplayCueOnActor(Burst) (Variables|FortMovementMode_BaseExtLogic|GetOwnerPawn) (Variables|FortMovementMode_ExtLogicFlyingJetpack|GetHitBonkCueTag) (Ability|GameplayCue|MakeGameplayCueParameters :NormalizedMagnitude 0.0 :RawMagnitude (Variables|/Script/BarnTowelRuntime.FortWeaponBarnTowel|GetForwardSpeed _asb_weapon_barn_towel) :Location _location :Normal _normal :Instigator (Variables|FortMovementMode_BaseExtLogic|GetOwnerPawn) :EffectCauser _hitactor :PhysicalMaterial _physmat :GameplayEffectLevel 1 :AbilityLevel 1 :TargetAttachComponent _hitcomponent :bReplicateLocationWhenUsingMinimalRepProxy false)))))

(fn PostProcessMovementVector
  (params (Move%20Status :type Fort%20Movement%20Mode%20Jetpack%20Flying%20Runtime%20Data%20Object%20Reference))
  (if (Variables|/Script/BarnTowelRuntime.FortMovementMode_BarnTowelFlyingRuntimeData|GetbIsStartingFlying Move%20Status)
    (return (Utilities|Operators|Add (Math|Vector|ClampVectorSize (Variables|Default|GetMovementVector) 0.0 1.0) (Math|Vector|MakeVector 0.0 0.0 (* (Utilities|RemapFloat (Class|FortJetpackFlyingBlueprintLibrary|GetFlyingDuration Move%20Status) (Math|Vector2D|MakeVector2D (- (ScalableFloat|GetValueatLevel (Variables|Balance|StartFlying|GetStartedFlyingTransitionDuration)) (ScalableFloat|GetValueatLevel (Variables|Balance|PostProcessMovementVector|GetPostProcessMovementVectorStartedFlyingTimeModifier))) (ScalableFloat|GetValueatLevel (Variables|Balance|StartFlying|GetStartedFlyingTransitionDuration))) (Math|Vector2D|MakeVector2D (ScalableFloat|GetValueatLevel (Variables|Balance|PostProcessMovementVector|GetPostProcessMovementVectorTotalFlyingTimeNewRangeX)) (ScalableFloat|GetValueatLevel (Variables|Balance|PostProcessMovementVector|GetPostProcessMovementVectorTotalFlyingTimeNewRangeY))) false) (- 1.0 (Math|Vector|VectorLength (Math|Vector|ClampVectorSize (Variables|Default|GetMovementVector) 0.0 1.0)))))))
    (else
      (return (CallFunction|ClamptoMaxHeight (Utilities|Operators|Add (Math|Vector|ClampVectorSize (Variables|Default|GetMovementVector) 0.0 1.0) (* (Variables|Default|GetCurrentNoiseOffset) (ScalableFloat|GetValueatLevel (Variables|Balance|PostProcessMovementVector|GetPostProcessMovementVectorCurrentNoiseOffsetMultiplier))) (Math|Vector|MakeVector 0.0 0.0 (* (Variables|Default|GetAscendMoveOffset) (Utilities|RemapFloat (Math|Vector|VectorLength (Math|Vector|ClampVectorSize (Variables|Default|GetMovementVector) 0.0 1.0)) "(X=0.000000,Y=1.000000)" (Math|Vector2D|MakeVector2D (ScalableFloat|GetValueatLevel (Variables|Balance|PostProcessMovementVector|GetPostProcessMovementVectorClampedLengthNewRangeX)) (ScalableFloat|GetValueatLevel (Variables|Balance|PostProcessMovementVector|GetPostProcessMovementVectorClampedLengthNewRangeY))) true)))) (Variables|GetMove%20Status))))))

(fn SetBoost
  (params (Boost :type Float%20%28double-precision%29))
  (bind _output_get (Variables|Default|SetBoost Boost)) ; TODO NATIVIZE -- Made it so Boost is only set at the end of hover (instead of constantly being set while in this state)
  (if (>= _output_get 1.0)
    (Variables|Default|SetBoostWasSetThisFrame true)))

(fn ShouldDoSpeedChanges
  (params (MoveStatus :type Fort%20Movement%20Mode%20Jetpack%20Flying%20Runtime%20Data%20Object%20Reference))
  (return (and (Math|Boolean|ORBoolean (Utilities|IsMoveInputActive MoveStatus) (Class|FortJetpackFlyingBlueprintLibrary|FlyingStatusHasAnyFlags (Variables|/Script/BarnTowelRuntime.FortMovementMode_BarnTowelFlyingRuntimeData|GetFlyingStatus MoveStatus) 32) (GameplayTags|HasMatchingGameplayTag (Variables|FortMovementMode_BaseExtLogic|GetOwnerPawn) (Variables|FortMovementMode_ExtLogicFlyingJetpack|GetRepulsorCannonTag))) (not (CallFunction|IsinInitialHover MoveStatus)))))

(fn TEMP_IsInChargingState
  (params (Move%20Status :type Fort%20Movement%20Mode%20Jetpack%20Flying%20Runtime%20Data%20Object%20Reference))
  (bind _asplayer_pawn_athena (Utilities|Casting|CastToPlayerPawn_Athena (Variables|FortMovementMode_BaseExtLogic|GetOwnerPawn)))
  (return (and (Math|Boolean|ORBoolean (Class|FortJetpackFlyingBlueprintLibrary|FlyingStatusHasAnyFlags (Variables|/Script/BarnTowelRuntime.FortMovementMode_BarnTowelFlyingRuntimeData|GetFlyingStatus Move%20Status) 16) (GameplayTags|HasMatchingGameplayTag _asplayer_pawn_athena (Variables|FortMovementMode_ExtLogicFlyingJetpack|GetIsInChargingStateRepulsorCannonTag)) false false (not (.bSuccess (Utilities|Casting|CastToB_Weapon_BarnTowel (Variables|/Script/FortniteGame.FortPawn|GetCurrentWeapon _asplayer_pawn_athena))))) (not (GameplayTags|HasAnyMatchingGameplayTags (Variables|FortMovementMode_BaseExtLogic|GetOwnerPawn) (Variables|FortMovementMode_ExtLogicFlyingJetpack|GetForceCrashDamagedTags)))) (not (GameplayTags|HasMatchingGameplayTag _asplayer_pawn_athena (Variables|FortMovementMode_ExtLogicFlyingJetpack|GetIsInChargingStateRepulsorCannonTag)))))

(fn ToggleInputMappingContext
  (params (Enable :type Boolean))
  (bind _cachedplayercontroller (Variables|Default|GetCachedPlayerController))
  (bind _returnvalue (PlayerController|LocalPlayerSubsystems|GetEnhancedInputLocalPlayerSubsystem _cachedplayercontroller))
  (bind _ownerpawn (Variables|FortMovementMode_BaseExtLogic|GetOwnerPawn))
  (bind _flyinginputcontext (Variables|Default|GetFlyingInputContext))
  (bind _flyinginputcontext_2 (Variables|Default|GetFlyingInputContext))
  (Utilities|IsValid _cachedplayercontroller
    (:Is%20Valid
      (if Enable
        (Input|AddMappingContext _returnvalue _flyinginputcontext 1000 "(bIgnoreAllPressedKeysUntilRelease=True,bForceImmediately=False,bNotifyUserSettings=False)")
        (Utilities|IsValid _ownerpawn
          (:Is%20Valid
            (Class|FortJetpackFlyingBlueprintLibrary|SetInputActionDisplayContext _ownerpawn _flyinginputcontext_2)))
        (else
          (Input|RemoveMappingContext _returnvalue _flyinginputcontext "(bIgnoreAllPressedKeysUntilRelease=True,bForceImmediately=False,bNotifyUserSettings=False)")
          (Utilities|IsValid _ownerpawn
            (:Is%20Valid
              (Class|FortJetpackFlyingBlueprintLibrary|ClearInputActionDisplayContext _ownerpawn _flyinginputcontext_2))))))))
