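; Gameplay ability that launches the player into jetpack flight with the BarnTowel weapon unless its cooldown tag is present.
; A state machine runs wind-up (vehicle eject, net sync), initial hover (starts flying), active thrust and off, swapping gameplay effects and timers.
; Active thrust starts a native destruction trace for smashing through things; entering off cleans up and ends the ability.
; The server replicates state changes, which clients apply after a short delay to respect prediction.
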(event Ability|EventActivateAbility
  ;[0.4838 1 0.2276] Activate Ability
  (|CallParentFunction)
  (Utilities|FlowControl|Sequence
    (:then_0
      (bind _asfort_player_pawn (Utilities|Casting|CastToFortPlayerPawn (Ability|GetActivatingPawn)))
      (bind _output_get (Variables|Default|SetPlayerPawn _asfort_player_pawn))
      (bind _asb_weapon_barn_towel (Utilities|Casting|CastToB_Weapon_BarnTowel (Variables|/Script/FortniteGame.FortPawn|GetCurrentWeapon _output_get)))
      (Variables|Default|SetWeapon%20Barn%20Towel _asb_weapon_barn_towel))
    (:then_1
      (Variables|Default|GetPlayerPawn
        (:then
          (Variables|Default|GetWeapon%20Barn%20Towel
            (:then
              (if (not (GameplayTags|HasMatchingGameplayTag (Variables|Default|GetPlayerPawn) "(TagName=\"Granted.Cooldown.BarnTowel\")"))
                (Default|SetRamState "NewEnumerator0")
                (Variables|Default|SetHasBlockBeenHandled false)
                (else
                  (Ability|EndAbility))))
            (:else
              _)))
        (:else
          _))
      (Ability|EndAbility))))

(event Custom|OnEarlyEjectComplete
  ;[0.1025 1 0.6739] On Enter State
  ;;[0.07432 0.3973 1] On Early Eject Complete
  (bind _asynctaskproxy (Ability|Tasks|WaitNetSync "BothWait")
    (:then
      (Variables|Transient|SetWaitEjectSyncAsyncTask _asynctaskproxy))
    (:OnSync
      (Development|PrintString (Utilities|String|Append (Utilities|GetDisplayName (Variables|Default|GetPlayerPawn)) "Movement - Early ejection complete."))
      (if (Math|Float|NearlyEqual(Float) (Fortnite|GetNumericCVar "Flying.BarnTowel.UseNativeLogic"))
        (Default|SetUseHighAccelerationTolerance true) ; Use High Acceleration Tolerance to prevent minor net corrections
      )
      (CallFunction|ApplyGameplayEffectSafe (Variables|Config|GetGE_BlockGrindingButDontCancel_Class) (Variables|Transient|GetGE_BlockGrindingButDontCancel))
      (bind _returnvalue (Utilities|Time|SetTimerbyEvent (EventDispatchers|CreateEvent :!event Custom|OnWindUpPhaseComplete) 0.01))
      (Variables|Transient|SetWindUpTimerHandle _returnvalue)
      (Utilities|FlowControl|Sequence
        (:then_0
          (CallFunction|ApplyGameplayEffectSafe (Variables|Config|GetGE_State_WindingUp_Class) (Variables|Transient|GetGE_State_WindingUp)))))))

(event Custom|OnWindUpPhaseComplete
  ;[0.1025 1 0.6739] On Enter State
  ;;[0.07432 0.3973 1] On Wind Up Phase Complete
  (Default|SetRamState "NewEnumerator3"))

(event Custom|HandleServerReplicatedState
  ;[0.2735 0.1586 1] Replicate State From Server
  (if (and (Utilities|Enum|NotEqual(Enum) (Variables|Transient|GetServerReplicatedState) (Variables|Default|GetPreviousLaunchState)) (Utilities|Enum|NotEqual(Enum) (Variables|Transient|GetServerReplicatedState) (Variables|Default|GetCurrentLaunchState)))
    (Default|SetRamState (Variables|Transient|GetServerReplicatedState)) ; Only overwrite the client state with the server state if it is transitioning to \nif the Previous and Current state on the client are different than the server state. \nWe check both because it is possible for the client to predictively transition into and \nout of the state that is being replicated from the server (i.e. client predicts Recovery, \nserver replicates Recovery, but client is in Off before the replicated event is received). \nIt is highly unlikely that the client transitions through two states before the replicated \nevent is received.
  ))

(event Custom|Client_ReplicateStateFromServer
  (params (ServerState :type Enum_LaunchState_BarnTowel%20Enum :default "NewEnumerator0"))
  ;[0.2735 0.1586 1] Replicate State From Server
  (Variables|Transient|SetServerReplicatedState ServerState)
  (bind _returnvalue (Utilities|Time|SetTimerbyEvent (EventDispatchers|CreateEvent :!event Custom|HandleServerReplicatedState) 0.1))
  (Variables|Transient|SetHandleServerReplicatedStateTimerHandle _returnvalue))

(event Ability|EventOnEndAbility
  (params (bWasCancelled :type Boolean))
  ;[1 0.5582 0.1752] End Ability
  (|CallParentFunction bWasCancelled)
  (Transformation|SetRelativeRotation (Variables|/Script/Engine.Character|GetMesh (Variables|Default|GetPlayerPawn)) "0.000000,-90.000000,0.000000" false false)
  (Default|SetRamState "NewEnumerator2") ; Always transition to Off when ending this ability \nand always end this ability when transitioning to Off.
  (CallFunction|ClearAllTimers))

(event Custom|StartDestructionTrace
  ;[1 0 0] Destruction Trace - Nativized
  (if (not (ScalableFloat|GetValueasBool (Variables|Default|GetSF_DisableSmashingThroughThings)))
    (CallFunction|NativeStartDestructionTrace)))

(event Custom|OnEnterRamState
  (params (CurrentState :type Enum_LaunchState_BarnTowel%20Enum :default "NewEnumerator0"))
  (bind (_returnvalue_roll _returnvalue_pitch _returnvalue_yaw) (Math|Rotator|BreakRotator (Transformation|GetActorRotation (Variables|Default|GetPlayerPawn))))
  (bind (_returnvalue_x _returnvalue_y _returnvalue_z) (Math|Vector|BreakVector (Transformation|GetActorLocation (Variables|Default|GetPlayerPawn))))
  ;[0.1025 1 0.6739] On Enter State
  (switch Utilities|FlowControl|Switch|SwitchonEnum_LaunchState_BarnTowel CurrentState
    (:NewEnumerator0
      (Ability|ApplyGameplayEffectToOwner "/BarnTowelGameplay/Gameplay/GameplayEffects/GE_BarnTowel_BlockAttackActivations_FromTakeOff.GE_BarnTowel_BlockAttackActivations_FromTakeOff_C")
      ;;[0.07432 0.3973 1] On State Changed To Wind Up
      (CallFunction|TryEarlyEject))
    (:NewEnumerator3
      (Utilities|FlowControl|Sequence
        (:then_0
          (Class|FortWeaponJetpack|StartFlying (Variables|Default|GetWeapon%20Barn%20Towel) (Utilities|Struct|MakeFortMovementModeFlyingCreationData _returnvalue_z (Math|Rotator|MakeRotator 0.0 0.0 _returnvalue_yaw)))
          (Development|PrintString "START CUSTOM FLYING" true true "(R=0.000000,G=0.660000,B=1.000000,A=1.000000)" 2e+01))
        (:then_1
          (Utilities|FlowControl|Delay (ScalableFloat|GetValueatLevel (Variables|Default|GetSF_WindUpHoverTime)))
          (Default|SetRamState "NewEnumerator1")
          (Ability|Tasks|WaitGameplayTagRemoveFromActor (Variables|Default|GetPlayerPawn) "(TagName=\"MovementMode.Flying\")"
            (:Removed
              (Ability|EndAbility))))))
    (:NewEnumerator1
      (Ability|RemoveGameplayEffectFromOwnerWithHandle (Variables|Transient|GetGE_State_WindingUp))
      (CallFunction|ApplyGameplayEffectSafe (Variables|Config|GetGE_State_ActiveThrust_Class) (Variables|Transient|GetGE_State_ActiveThrust))
      (CallFunction|StartDestructionTrace) ; Needs to happen last; this node can sometimes result\nin changing the state
      (if (ScalableFloat|GetValueasBool (Variables|Default|GetSF_RemoveWeaponSwitchBlockAfterInitialHover))
        (GameplayEffects|RemoveActiveGameplayEffectbySourceEffect (Ability|GetFortAbilitySystemComponent (Variables|Default|GetPlayerPawn)) "/BarnTowelGameplay/Gameplay/GameplayEffects/GE_BlockSwitchingWeapons_BarnTowel.GE_BlockSwitchingWeapons_BarnTowel_C" 0 -1)))
    (:NewEnumerator2
      ;;[0.4235 1 0.04638] On State Changed To Off
      (Utilities|FlowControl|Sequence
        (:then_0
          (if (Math|Float|NearlyEqual(Float) (Fortnite|GetNumericCVar "Flying.BarnTowel.UseNativeLogic"))
            (Default|SetUseHighAccelerationTolerance false))
          (CallFunction|StopDestructionTrace)
          (CallFunction|ResetIgnoredCollision))
        (:then_1
          (Utilities|Time|ClearandInvalidateTimerbyHandle (Variables|Transient|GetHandleClientReplicatedBlockingHitTimerHandle))
          (Utilities|Time|ClearandInvalidateTimerbyHandle (Variables|Transient|GetHandleServerReplicatedStateTimerHandle))
          (Ability|RemoveGameplayEffectFromOwnerWithHandle (Variables|Transient|GetGE_State_WindingUp))
          (Ability|RemoveGameplayEffectFromOwnerWithHandle (Variables|Transient|GetGE_State_ActiveThrust))
          (Ability|RemoveGameplayEffectFromOwnerWithHandle (Variables|Transient|GetGE_BlockGrindingButDontCancel)) ; Make sure any existing one is removed
          (Ability|RemoveGameplayEffectFromOwnerWithHandle (Variables|Default|GetGC_SmashingFX))
          (Components|DestroyComponent (Variables|Default|GetTelegraphNS_OnWater))
          (Actor|DestroyActor (Variables|Default|GetTelegraphBP)))
        (:then_2
          ;;;[1 0.581 0] Log and End Ability
          (Ability|EndAbility) ; Always transition to Off when ending this ability \nand always end this ability when transitioning to Off.
        )))))

(event Custom|StopDestructionTrace
  ;[1 0 0] Destruction Trace - Nativized
  (CallFunction|NativeStopDestructionTrace))

(event Custom|ResetIgnoredCollision
  ;[1 0 0] Destruction Trace - Nativized
  ;;[1 0.06918 0.5615] Nativized - Reset ignored collision. Called on state changed to Off. This is a failsafe to ensure that if something wasn't destroyed, but had its collision ignored, we unignore it here to prevent exploits and bugs.
  (CallFunction|ResetDestructionIgnoredCollision))

(event Custom|SetGrindingBlocked
  (params (bIsBlocked :type Boolean))
  (bind _blockgrindrailstag (Variables|Config|GetBlockGrindRailsTag))
  ;[1 0.2517 0.7354] Block Grinding
  (bind _playerpawn (Variables|Default|GetPlayerPawn)
    (:then
      (if bIsBlocked
        (if (not (Variables|Transient|Grinding|GetbHasAppliedBlockGrindRailsTag))
          (Ability|GameplayEffect|AddLooseGameplayTags _playerpawn _blockgrindrailstag)
          (Variables|Transient|Grinding|SetbHasAppliedBlockGrindRailsTag true))
        (elif (Variables|Transient|Grinding|GetbHasAppliedBlockGrindRailsTag)
          (Ability|GameplayEffect|RemoveLooseGameplayTags _playerpawn _blockgrindrailstag)
          (Variables|Transient|Grinding|SetbHasAppliedBlockGrindRailsTag false))))
    (:else)))

(event Custom|TryEarlyEject
  (bind _playerpawn (Variables|Default|GetPlayerPawn))
  (bind _returnvalue (Vehicle|GetVehicle _playerpawn))
  (bind _nativedestructiontraceignoredactors (Variables|FortGameplayAbility_LaunchJetpack|GetNativeDestructionTraceIgnoredActors))
  ;[0.07432 0.3973 1] Try Early Eject From other movement stuff
  (bind _playerpawn_2 (Variables|Default|GetPlayerPawn)
    (:then
      (if (GameplayTags|HasMatchingGameplayTag _playerpawn_2 "(TagName=\"Gameplay.Status.InVehicle\")")
        (Development|PrintString (Utilities|String|Append (Utilities|GetDisplayName (Variables|Default|GetPlayerPawn)) "Movement - Ejecting from vehicle."))
        (Utilities|IsValid _returnvalue
          (:Is%20Valid
            (Utilities|Array|Add _nativedestructiontraceignoredactors _returnvalue)
            (Utilities|Array|AppendArray _nativedestructiontraceignoredactors (Vehicle|GetAllPassengers _returnvalue))
            (bind _returnvalue_2 (Vehicle|EjectPlayerPawn _returnvalue _playerpawn))
            (if _returnvalue_2
              (bind _asynctaskproxy (Ability|Tasks|WaitGameplayTagRemove "(TagName=\"Gameplay.Status.InVehicle\")" 0 true)
                (:then
                  (Variables|Transient|SetWaitEjectFromVehicleAsyncTask _asynctaskproxy))
                (:Removed
                  (CallFunction|OnEarlyEjectComplete)))
              (else
                (Ability|EndAbility)))))
        (elif (GameplayTags|HasMatchingGameplayTag (Variables|Default|GetPlayerPawn) "(TagName=\"Granted.Riding.Player.IsRiding\")")
          (Development|PrintString (Utilities|String|Append (Utilities|GetDisplayName (Variables|Default|GetPlayerPawn)) "Movement - Ejecting from riding animal."))
          (else
            (CallFunction|OnEarlyEjectComplete)))))
    (:else)))

(event Custom|OnExitRamState
  (params (PreviousState :type Enum_LaunchState_BarnTowel%20Enum :default "NewEnumerator0"))
  ;[1 0.4181 0.09459] On Exit State
  (switch Utilities|FlowControl|Switch|SwitchonEnum_LaunchState_BarnTowel PreviousState
    (:NewEnumerator0
      ;;[0.07432 0.3973 1] On State Changed From Winding Up
      (Utilities|Time|ClearandInvalidateTimerbyHandle (Variables|Transient|GetWindUpTimerHandle)))
    (:NewEnumerator3
      ;;[0.02924 0.08066 1] On State Changed From InitialHover
      (Utilities|Time|ClearandInvalidateTimerbyHandle (Variables|Transient|GetSlowAccelerationTimerHandle))
      (GameplayTasks|EndTask (Variables|Transient|GetWaitLaunchDirectionReplicatedAsyncTask)))
    (:NewEnumerator1
      ;;[0.7199 0.1195 1] On State Changed From Flying
      (GameplayTasks|EndTask (Variables|Transient|GetRootMotionAsyncTask))
      (Ability|RemoveGameplayEffectFromOwnerWithHandle (Variables|Transient|GetGE_State_ActiveThrust))
      (CallFunction|SetGrindingBlocked false)
      (Ability|RemoveGameplayEffectFromOwnerWithHandle (Variables|Transient|GetGE_BlockGrindingButDontCancel)) ; Make sure any existing one is removed
    )))

(event Custom|TryLateEject
  ;[1 0.2746 0.9747] Try Late Eject
  ;;[1 1 1] Comment
  (bind _playerpawn (Variables|Default|GetPlayerPawn)
    (:then
      (if (GameplayTags|HasMatchingGameplayTag _playerpawn "(TagName=\"MovementMode.ZipliningOnSpline\")")
        (Development|PrintString (Utilities|String|Append (Utilities|GetDisplayName (Variables|Default|GetPlayerPawn)) "Movement - Ejecting from zipline."))
        (Zipline|EndZiplining (Variables|Default|GetPlayerPawn) false false)
        (elif (GameplayTags|HasMatchingGameplayTag (Variables|Default|GetPlayerPawn) "(TagName=\"MovementMode.Grinding\")")
          (CallFunction|SetGrindingBlocked true)
          (Pawn|Components|CharacterMovement|SetMovementMode (Variables|/Script/Engine.Character|GetCharacterMovement (Variables|Default|GetPlayerPawn)) "MOVE_Falling" 0))))
    (:else)))

(fn ApplyGameplayEffectSafe
  (params (Class :type Gameplay%20Effect%20Class%20Reference) (Handle :type Active%20Gameplay%20Effect%20Handle%20Structure%20%28by%20ref%29))
  (Utilities|IsValid (Ability|GameplayEffect|GetGameplayEffectfromActiveEffectHandle Handle)
    (:Is%20Not%20Valid
      (bind _returnvalue (Ability|ApplyGameplayEffectToOwner Class))
      (Variables|SetBy-RefVar Handle _returnvalue))) ; If this is valid then it's trying to double-apply the\nGE, which is an error.
)

(fn ClearAllTimers
  (for _array_element (Utilities|Array|MakeArray (Variables|Transient|GetWindUpTimerHandle) (Variables|Transient|GetSlowAccelerationTimerHandle) (Variables|Transient|GetDestructionTraceTimerHandle) (Variables|Transient|GetGravityIncreaseTimerHandle) (Variables|Transient|GetAirControlIncreaseTimerHandle) (Variables|Transient|GetRecoveryTimerHandle) (Variables|Transient|GetBrakingTimerHandle) (Variables|Transient|GetTimeDilationEaseInTimerHandle) (Variables|Transient|GetLookaheadTraceTimerHandle) (Variables|Transient|GetSlowAccelerationLockInLaunchDirectionTimerHandle) (Variables|Transient|GetPredictLaunchDirectionTimerHandle) (Variables|Transient|GetPostAOEPauseTimerHandle) (Variables|Transient|GetHandleClientReplicatedBlockingHitTimerHandle) (Variables|Transient|GetHandleServerReplicatedStateTimerHandle))
    (Utilities|Time|ClearandInvalidateTimerbyHandle _array_element)))

(fn ForceNetUpdate
  (bind _playerpawn (Variables|Default|GetPlayerPawn)
    (:then
      (Networking|ForceNetUpdate _playerpawn))
    (:else)))

(fn GetCustomAbilitySourceTransform
  (return (Variables|Default|GetCurrentCustomTargetingTransform)))

(fn IsHitAgainstWall
  (params (HitResult :type Hit%20Result%20Structure))
  (bind (_bblockinghit _binitialoverlap _time _distance _location _impactpoint _normal _impactnormal) (Collision|BreakHitResult HitResult))
  (return (>= (Math|Trig|Acos(Degrees) (Math|Vector|DotProduct _impactnormal "0.000000,0.000000,1.000000")) 5e+01)))

(fn PlayDestructionEffects
  (params (Damaged%20A%20Build :type Boolean) (Damaged%20A%20Vehicle :type Boolean) (Damaged%20A%20Pawn :type Boolean))
  (Variables|Default|GetPlayerPawn
    (:then
      (if (and (Math|Boolean|ORBoolean Damaged%20A%20Build Damaged%20A%20Vehicle Damaged%20A%20Pawn) (> (- (Utilities|Time|GetGameTimeinSeconds) (Variables|Transient|GetLastHitSomethingGCTime)) (Variables|Default|GetMinTimeBetweenHitSomethingGC)))
        (GameplayCue|ExecuteGameplayCueOnActor(Burst) (Variables|Default|GetPlayerPawn) (Variables|Config|GetGC_HitSomething) (Ability|GameplayCue|MakeGameplayCueParameters (Math|Conversions|ToFloat(Integer) (Math|Integer|SelectInt 0 1 (Variables|GetDamaged%20A%20Pawn)))))
        (Variables|Transient|SetLastHitSomethingGCTime (Utilities|Time|GetGameTimeinSeconds))))
    (:else)))

(fn SetRamState
  (params (NewState :type Enum_LaunchState_BarnTowel%20Enum :default "NewEnumerator0"))
  (if (Utilities|Enum|NotEqual(Enum) (Variables|Default|GetCurrentLaunchState) (Variables|GetNewState))
    (Variables|Default|SetPreviousLaunchState (Variables|Default|GetCurrentLaunchState))
    (Variables|FortGameplayAbility_LaunchJetpack|SetNativeCurrentLaunchState (Utilities|Enum|BytetoEnumEJetpackLaunchState (Math|Conversions|ToByte(Integer) (Math|Conversions|ToInteger(Byte) (Variables|GetNewState)))))
    (Variables|Default|SetCurrentLaunchState (Variables|GetNewState))
    (CallFunction|OnExitRamState (Variables|Default|GetPreviousLaunchState))
    (CallFunction|OnEnterRamState (Variables|Default|GetCurrentLaunchState))
    (Ability|HasAuthority
      (:True
        (CallFunction|ClientReplicateStatefromServer (Variables|Default|GetCurrentLaunchState))
        (Default|ForceNetUpdate)))))

(fn SetUseHighAccelerationTolerance
  (params (bUseHighAccelerationTolerance :type Boolean))
  (bind _playerpawn (Variables|Default|GetPlayerPawn)
    (:then
      (bind _asfort_movement_comp_character (Utilities|Casting|CastToFortMovementComp_Character (Variables|/Script/Engine.Character|GetCharacterMovement _playerpawn)))
      (NetTolerance|SetUseHighAccelerationTolerance _asfort_movement_comp_character bUseHighAccelerationTolerance "RocketRamMovement"))
    (:else)))
