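; Secondary-input gameplay ability for BarnTowel that triggers the Repulsor Cannon (Unibeam) by sending a gameplay event to the pawn.
; Unless a scalable float disables it while flying, activation requires the in-inventory combo tag or a gauntlet-bypass setting, plus no cooldown tag.
; Blocked attempts play a local failure cue and report cooldown or missing weapon, with an extra failed-activation event for the latter.
; Failure overrides substitute the missing-weapon reason when the inventory tag is absent.
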(event Ability|EventOnAbilityInputReleased
  (|CallParentFunction))

(event Ability|EventActivateAbility
  (bind _returnvalue (Ability|GetActivatingPawn))
  (bind _returnvalue_2 (Ability|GetActivatingPawn))
  (|CallParentFunction)
  (if (not (Class|AthenaFunctionLibrary|RowtoBool (Variables|Default|GetSF_FullyDisableUnibeamWhileFlying)))
    (bind _returnvalue_3 (GameplayTags|HasMatchingGameplayTag(Message) (Ability|GetActivatingPawn) "(TagName=\"Granted.HighTower.Power.Tomato.Combo.InInventory\")"))
    (if (or _returnvalue_3 (ScalableFloat|GetValueasBool (Variables|Default|GetSF_AllowUnibeamWithoutGauntlets)))
      (bind _returnvalue_4 (GameplayTags|HasMatchingGameplayTag(Message) (Ability|GetActivatingPawn) "(TagName=\"Granted.Cooldown.HighTower.Tomato.RepulsorCannon\")"))
      (if _returnvalue_4
        (GameplayCue|ExecuteGameplayCue%28Burst%29Local _returnvalue (Variables|Default|GetGC_FailedUnibeamActivation) (Ability|GameplayCue|MakeGameplayCueParameters 0.0))
        (bind _asfort_player_pawn (Utilities|Casting|CastToFortPlayerPawn _returnvalue)
          (:then
            (FailedAction|UnabletoPerformAction _asfort_player_pawn true (GameplayTags|MakeLiteralGameplayTagContainer "(GameplayTags=((TagName=\"Athena.Ability.Failed.Cooldown\")))")))
          (:CastFailed
            _))
        (Ability|EndAbility)
        (else
          (Ability|SendGameplayEventtoActor (Ability|GetActivatingPawn) "(TagName=\"Event.Athena.BarnTowel.TriggerRepulsorCannon\")")
          (Ability|EndAbility)))
      (else
        (Ability|SendGameplayEventtoActor _returnvalue_2 "(TagName=\"Event.Athena.BarnTowel.Unibeam.FailedActivation\")")
        (GameplayCue|ExecuteGameplayCue%28Burst%29Local _returnvalue_2 (Variables|Default|GetGC_FailedUnibeamActivation) (Ability|GameplayCue|MakeGameplayCueParameters 0.0))
        (bind _asfort_player_pawn_2 (Utilities|Casting|CastToFortPlayerPawn (Ability|GetActivatingPawn))
          (:then
            (FailedAction|UnabletoPerformAction _asfort_player_pawn_2 false (GameplayTags|MakeGameplayTagContainerfromTag "(TagName=\"Athena.Ability.Failed.BarnTowel.MissingWeapon\")")) ; show reason for failing on screen if one is given
          )
          (:CastFailed
            _))
        (Ability|EndAbility)))))

(fn OverrideFailedReason
  (params (FailedReason :type Gameplay%20Tag%20Container%20Structure%20%28by%20ref%29))
  (bind _returnvalue (GameplayTags|HasMatchingGameplayTag(Message) (Ability|GetActivatingPawn) "(TagName=\"Granted.HighTower.Power.Tomato.Combo.InInventory\")"))
  (if _returnvalue
    (return FailedReason)
    (else
      (return "(GameplayTags=((TagName=\"Athena.Ability.Failed.BarnTowel.MissingWeapon\")))"))))

(fn ShouldPlayFailedMontage
  (params (FailedReason :type Gameplay%20Tag%20Container%20Structure%20%28by%20ref%29))
  (return true))
